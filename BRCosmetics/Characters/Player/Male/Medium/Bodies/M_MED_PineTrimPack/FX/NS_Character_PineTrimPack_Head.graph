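# Niagara cosmetic effect for the PineTrimPack character's head, producing fiery orange lightning and glows at a sampled socket.
# A mesh emitter spawns randomly oriented, scaled lightning, feeding scale, inversion and fade curves into dynamic material parameters.
# Three sprite emitters layer short-lived glows on the same offset socket, using camera offsets.
# All four spawn rates follow the system Timeline's normalized time, tapering to zero; user variables expose reactivity and front-end state.
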NiagaraSystem /BRCosmetics/Characters/Player/Male/Medium/Bodies/M_MED_PineTrimPack/FX/NS_Character_PineTrimPack_Head
dependencies {
 usedDataInterfaces [&NiagaraDataInterfaceCurve, &NiagaraDataInterfaceColorCurve, &NiagaraDataInterfaceSocketReader]
 usedDynamicInputs [&InvertBool]
 usedModules [
  &self:SystemSpawnScript.SystemScriptSource.SystemScriptGraph.NiagaraNodeAssignment_1.SetVariables_#1
  &SystemState
  /Game/Effects/Niagara/Modules/Logic/UpdateValueStateChange
  &Timeline
  &self:SystemSpawnScript.SystemScriptSource.SystemScriptGraph.NiagaraNodeAssignment_2.SetVariables_#2
  &self:SystemSpawnScript.SystemScriptSource.SystemScriptGraph.NiagaraNodeAssignment_0.SetVariables_#3
 ]
}
emitters {
 LowerArms003 {
  emitterUpdateScript [
   &EmitterState(
    Inactive Response=::NewEnumerator1
    Life Cycle Mode=::NewEnumerator1
    Loop Behavior=::NewEnumerator1
    Loop Duration Mode=::NewEnumerator1
    UseLoopDelay=false
   )
   &SpawnRate(
    SpawnRate=&FloatFromCurve(
     FloatCurve=DataInterface() {propertyValues {Curve {keys [
         {tangentMode=RCTM_User, time=0, value=1}
         {arriveTangent=-0.0405, interpMode=RCIM_Cubic, leaveTangent=-0.0405, tangentMode=RCTM_User, time=0.7009, value=1}
         {arriveTangent=-8.0089, interpMode=RCIM_Cubic, leaveTangent=-8.0089, tangentMode=RCTM_User, time=1, value=0}
        ]}}}
     CurveIndex=$System.Timeline.NormalizedTime
     Scale Curve=8
    )
    Use Spawn Probability=false
   )
  ]
  particleSpawnScript [
   &InitializeParticle(
    Lifetime=&RandomRangeFloat(Minimum=0.6, Maximum=1)
    UsePositionOffset=false
    Write Material Random=true
   )
   /SetVariables(Particles.InvertLightning=true, Particles.MirrorLightning:NiagaraBool=&RandomBool()) {isSetParameters=true}
   /SetVariables(Particles.LightningIndex:NiagaraInt32=&RandomRangeInt(Minimum=0, Maximum=0)) {isSetParameters=true}
   /SetVariables(Particles.Scale=&Multiply_VectorByFloat(
     Vector=&RandomRangeVector(Minimum=<1 1 1>, Maximum=<1.1 1.1 1.1>)
     Float=&RandomRangeFloat(Minimum=0.2, Maximum=0.3)
    )) {isSetParameters=true}
   &InitialMeshOrientation(
    Mesh Orientation Mode=::NewEnumerator7
    Rotation=&RandomRangeVector(Minimum=<0 0 0>, Maximum=<1 1 1>)
    Use Rotation Vector=true
   ) {module=InitialMeshOrientation001}
  ]
  particleUpdateScript [
   &SocketLocation(
    SamplePosOffset=true
    Sampled Position Offset=<10 0 0>
    Socket Reader=$System.Socket Reader Interface
    Socket Sampling Mode=::NewEnumerator3
    VelocitySpeedLimit=false
   )
   &DynamicMaterialParameters(
    Index 0 Param 1=&RandomRangeFloat(Minimum=1, Maximum=2)
    Index 0 Param 2=&Divide_Float(A=&RandomRangeFloat(Minimum=1, Maximum=1.5), B=1)
    Index 0 Param 3=&FloatFromCurve(FloatCurve=DataInterface() {propertyValues {Curve {keys [<at 0 1>,<at 0.1 0.375>,<at 0.5 0.05>,<at 1 0>]}}})
    Index 0 Param 4=&MakeCustomFloatFromBool(Bool=$Particles.InvertLightning, True Float=1, False Float=0)
    Index 1 Param 1=&MakeFloatFromVector(Vector=$Particles.Scale, Channel=::NewEnumerator0)
    Index 1 Param 2=&MakeFloatFromVector(Vector=$Particles.Scale, Channel=::NewEnumerator1)
    Index 1 Param 3=&MakeFloatFromVector(Vector=$Particles.Scale, Channel=::NewEnumerator2)
    Module.Index0Param0Write=true
    Module.Index0Param1Write=true
    Module.Index0Param2Write=true
    Module.Index0Param3Write=true
    Module.Index1Param0Write=true
    Module.Index1Param1Write=true
    Module.Index1Param2Write=true
    Module.Index1Param3Write=true
    Write Parameter Index 1=true
   )
   &ParticleState()
   &CameraOffset(Camera Offset Amount=4)
   &Color(Color=&MultiplyLinearColor(
     Linear Color=&ColorFromCurve(ColorCurve=DataInterface() {propertyValues {
        AlphaCurve {keys [
          <at 0 0>
          {arriveTangent=0.8076, interpMode=RCIM_Cubic, leaveTangent=0.8076, time=0.0998, value=1}
          <at 0.6191 0.5>
          <at 0.9862 0 RCIM_Cubic>
         ]}
        BlueCurve {keys [<at 0 0 RCIM_Cubic>,<at 0.2919 0>,<at 1 0 RCIM_Cubic>]}
        GreenCurve {keys [<at 0 39.6 RCIM_Cubic>,<at 0.2919 1.8>,<at 1 0.25 RCIM_Cubic>]}
        RedCurve {keys [<at 0 66 RCIM_Cubic>,<at 0.2919 3>,<at 1 1 RCIM_Cubic>]}
       }})
     Scale Factor=&MakeVector4FromVectorAndFloat(
      "Vector (RGB)"=&VectorFromFloat(Value=&Multiply_Float(A=2, B=1))
      "Float (Alpha)"=&RandomRangeFloat(Minimum=0.1, Maximum=1, Evaluation Type=::NewEnumerator1)
     )
    ))
   &FN_Mesh_Attributes(
    Color=&MultiplyLinearColor(
     Linear Color=&ColorFromCurve(ColorCurve=DataInterface() {propertyValues {
        AlphaCurve {keys [
          <at 0 0>
          {arriveTangent=1.9183, interpMode=RCIM_Cubic, leaveTangent=1.9183, time=0.0998, value=1}
          <at 0.5213 1>
          <at 0.8444 0.3>
          <at 1 0 RCIM_Cubic>
         ]}
        BlueCurve {keys [<at 0 0 RCIM_Cubic>,<at 0.2919 0>,<at 1 0 RCIM_Cubic>]}
        GreenCurve {keys [<at 0 0.06 RCIM_Cubic>,<at 0.2919 0.03>,<at 1 0.0217 RCIM_Cubic>]}
        RedCurve {keys [<at 0 0.1 RCIM_Cubic>,<at 0.2919 0.05>,<at 1 0.05 RCIM_Cubic>]}
       }})
     Scale Factor=<1 1 1 0.35>
    )
    Create Color Attribute=true
    Write to Module Namespace=true
   )
   &FN_Mesh_Attributes(
    Color=&MultiplyLinearColor(
     Linear Color=&ColorFromCurve(ColorCurve=DataInterface() {propertyValues {
        AlphaCurve {keys [<at 0 0>,<at 0.0998 1 RCIM_Cubic>,<at 0.75 0.3>,<at 1 0 RCIM_Cubic>]}
        BlueCurve {keys [<at 0 0 RCIM_Cubic>,<at 0.2919 0>,<at 1 0 RCIM_Cubic>]}
        GreenCurve {keys [<at 0 39.6 RCIM_Cubic>,<at 0.2919 1.8>,<at 1 0.6 RCIM_Cubic>]}
        RedCurve {keys [<at 0 66 RCIM_Cubic>,<at 0.2919 3>,<at 1 1 RCIM_Cubic>]}
       }})
     Scale Factor=&MakeVector4FromVectorAndFloat(
      "Vector (RGB)"=<1 1 1>
      "Float (Alpha)"=&RandomRangeFloat(Minimum=0, Maximum=0.3, Evaluation Type=::NewEnumerator1)
     )
    )
    Create Color Attribute=true
    Write to Module Namespace=true
   ) {module=FN_Mesh_Attributes001}
   &FN_Mesh_Attributes(Create Color Attribute=true, Write to Module Namespace=true) {module=FN_Mesh_Attributes002}
  ]
  rendererClasses [&NiagaraMeshRendererProperties]
  renderers [{rendererClass=&NiagaraMeshRendererProperties, rendererIndex=0}]
 }
 NE_Glow {
  emitterUpdateScript [
   &EmitterState(Life Cycle Mode=::NewEnumerator1, Loop Duration=1, UseLoopDelay=false)
   &SpawnBurst_Instantaneous(Spawn Count=1, Use Spawn Probability=false) {enabled=false}
   &SpawnRate(
    SpawnRate=&FloatFromCurve(
     FloatCurve=DataInterface() {propertyValues {Curve {keys [
         {tangentMode=RCTM_User, time=0, value=1}
         {arriveTangent=-0.0405, interpMode=RCIM_Cubic, leaveTangent=-0.0405, tangentMode=RCTM_User, time=0.7009, value=1}
         {arriveTangent=-8.0089, interpMode=RCIM_Cubic, leaveTangent=-8.0089, tangentMode=RCTM_User, time=1, value=0}
        ]}}}
     CurveIndex=$System.Timeline.NormalizedTime
     Scale Curve=10
    )
    Use Spawn Probability=false
   )
  ]
  particleSpawnScript [
   /Niagara/Modules/Spawn/Initialization/InitializeParticle(
    Lifetime=5
    Module.Write Color=true
    Module.Write Lifetime=true
    Module.Write Mass=true
    Module.Write Material Random=true
    Module.Write Position=false
    Module.Write Scale=false
    Module.Write SpriteRotation=false
    Module.Write SpriteSize=true
    Position=&SimulationPosition()
    Sprite Size=<50 50>
   ) {enabled=false}
   /Niagara/Modules/Spawn/Initialization/V2/InitializeParticle(
    Lifetime=&RandomRangeFloat(Minimum=0.5, Maximum=0.8)
    Sprite Size Mode=::NewEnumerator3
    Uniform Sprite Size=&RandomRangeFloat(Minimum=5, Maximum=10)
    UsePositionOffset=false
    Write Material Random=true
   ) {module=InitializeParticle001}
  ]
  particleUpdateScript [
   &ParticleState()
   &SocketLocation(
    SamplePosOffset=true
    Sampled Position Offset=<10 0 0>
    Socket Reader=$System.Socket Reader Interface
    Socket Sampling Mode=::NewEnumerator3
    VelocitySpeedLimit=false
   )
   &ShapeLocation(
    MaskSpawnGroup=false
    Override Seed=false
    Shape Origin=$Output.SocketLocation.SampledPosition
    Sphere Radius=&RandomRangeFloat(Minimum=0, Maximum=1)
    UseOwnerScale=true
   )
   &SolveForcesAndVelocity(Clamp Velocity=false, Limit Acceleration=false)
   &CameraOffset(Camera Offset Amount=10)
   &Color() {enabled=false}
   &Color(Color=&ColorFromCurve(ColorCurve=DataInterface() {propertyValues {
       AlphaCurve {keys [<at 0 0>, {arriveTangent=0.3873, interpMode=RCIM_Cubic, leaveTangent=0.3873, time=0.0998, value=1}, <at 0.7746 0.3>, <at 1 0 RCIM_Cubic>]}
       BlueCurve {keys [<at 0 0 RCIM_Cubic>,<at 0.2919 0>,<at 1 0.0833 RCIM_Cubic>]}
       GreenCurve {keys [<at 0 20.6613 RCIM_Cubic>,<at 0.2919 0.9>,<at 1 0 RCIM_Cubic>]}
       RedCurve {keys [<at 0 66 RCIM_Cubic>,<at 0.2919 3>,<at 1 1 RCIM_Cubic>]}
      }})) {module=Color001}
  ]
  rendererClasses [&NiagaraSpriteRendererProperties]
  renderers [{rendererClass=&NiagaraSpriteRendererProperties}]
 }
 NE_Glow001 {
  emitterUpdateScript [
   &EmitterState(Life Cycle Mode=::NewEnumerator1, Loop Duration=1, UseLoopDelay=false)
   &SpawnBurst_Instantaneous(Spawn Count=1, Use Spawn Probability=false) {enabled=false}
   &SpawnRate(
    SpawnRate=&FloatFromCurve(
     FloatCurve=DataInterface() {propertyValues {Curve {keys [
         {tangentMode=RCTM_User, time=0, value=1}
         {arriveTangent=-0.0405, interpMode=RCIM_Cubic, leaveTangent=-0.0405, tangentMode=RCTM_User, time=0.7009, value=1}
         {arriveTangent=-8.0089, interpMode=RCIM_Cubic, leaveTangent=-8.0089, tangentMode=RCTM_User, time=1, value=0}
        ]}}}
     CurveIndex=$System.Timeline.NormalizedTime
     Scale Curve=5
    )
    Use Spawn Probability=false
   )
  ]
  particleSpawnScript [
   /Niagara/Modules/Spawn/Initialization/InitializeParticle(
    Lifetime=5
    Module.Write Color=true
    Module.Write Lifetime=true
    Module.Write Mass=true
    Module.Write Material Random=true
    Module.Write Position=false
    Module.Write Scale=false
    Module.Write SpriteRotation=false
    Module.Write SpriteSize=true
    Position=&SimulationPosition()
    Sprite Size=<50 50>
   ) {enabled=false}
   /Niagara/Modules/Spawn/Initialization/V2/InitializeParticle(
    Lifetime=&RandomRangeFloat(Minimum=0.6, Maximum=0.8)
    Sprite Size Mode=::NewEnumerator3
    Uniform Sprite Size=&RandomRangeFloat(Minimum=10, Maximum=10)
    UsePositionOffset=false
    Write Material Random=true
   ) {module=InitializeParticle001}
  ]
  particleUpdateScript [
   &ParticleState()
   &SocketLocation(
    SamplePosOffset=true
    Sampled Position Offset=<10 0 0>
    Socket Reader=$System.Socket Reader Interface
    Socket Sampling Mode=::NewEnumerator3
    VelocitySpeedLimit=false
   )
   &ShapeLocation(
    MaskSpawnGroup=false
    Override Seed=false
    Shape Origin=$Output.SocketLocation.SampledPosition
    Sphere Radius=&RandomRangeFloat(Minimum=0, Maximum=2)
    UseOwnerScale=true
   )
   &SolveForcesAndVelocity(Clamp Velocity=false, Limit Acceleration=false)
   &CameraOffset(Camera Offset Amount=20)
   &Color() {enabled=false}
   &Color(Color=&ColorFromCurve(ColorCurve=DataInterface() {propertyValues {
       AlphaCurve {keys [<at 0 0>,<at 0.0998 0.1 RCIM_Cubic>,<at 1 0 RCIM_Cubic>]}
       BlueCurve {keys [<at 0 0 RCIM_Cubic>,<at 0.2919 0>,<at 1 0.0833 RCIM_Cubic>]}
       GreenCurve {keys [<at 0 20.6613 RCIM_Cubic>,<at 0.2919 0.9>,<at 1 0 RCIM_Cubic>]}
       RedCurve {keys [<at 0 66 RCIM_Cubic>,<at 0.2919 3>,<at 1 1 RCIM_Cubic>]}
      }})) {module=Color001}
  ]
  rendererClasses [&NiagaraSpriteRendererProperties]
  renderers [{rendererClass=&NiagaraSpriteRendererProperties}]
 }
 NE_Glow002 {
  emitterUpdateScript [
   &EmitterState(Life Cycle Mode=::NewEnumerator1, Loop Duration=1, UseLoopDelay=false)
   &SpawnBurst_Instantaneous(Spawn Count=1, Use Spawn Probability=false) {enabled=false}
   &SpawnRate(
    SpawnRate=&FloatFromCurve(
     FloatCurve=DataInterface() {propertyValues {Curve {keys [
         {tangentMode=RCTM_User, time=0, value=1}
         {arriveTangent=-0.0405, interpMode=RCIM_Cubic, leaveTangent=-0.0405, tangentMode=RCTM_User, time=0.7009, value=1}
         {arriveTangent=-8.0089, interpMode=RCIM_Cubic, leaveTangent=-8.0089, tangentMode=RCTM_User, time=1, value=0}
        ]}}}
     CurveIndex=$System.Timeline.NormalizedTime
     Scale Curve=5
    )
    Use Spawn Probability=false
   )
  ]
  particleSpawnScript [
   /Niagara/Modules/Spawn/Initialization/InitializeParticle(
    Lifetime=5
    Module.Write Color=true
    Module.Write Lifetime=true
    Module.Write Mass=true
    Module.Write Material Random=true
    Module.Write Position=false
    Module.Write Scale=false
    Module.Write SpriteRotation=false
    Module.Write SpriteSize=true
    Position=&SimulationPosition()
    Sprite Size=<50 50>
   ) {enabled=false}
   /Niagara/Modules/Spawn/Initialization/V2/InitializeParticle(
    Color=<rgba 200 105 10 1>
    Lifetime=&RandomRangeFloat(Minimum=0.2, Maximum=0.3)
    Sprite Size Mode=::NewEnumerator3
    Uniform Sprite Size=&RandomRangeFloat(Minimum=3, Maximum=3)
    UsePositionOffset=false
    Write Material Random=true
   ) {module=InitializeParticle001}
  ]
  particleUpdateScript [
   &ParticleState()
   &SocketLocation(
    SamplePosOffset=true
    Sampled Position Offset=<10 0 0>
    Socket Reader=$System.Socket Reader Interface
    Socket Sampling Mode=::NewEnumerator3
    VelocitySpeedLimit=false
   )
   &SolveForcesAndVelocity(Clamp Velocity=false, Limit Acceleration=false)
   &CameraOffset(Camera Offset Amount=15)
   &Color() {enabled=false}
  ]
  rendererClasses [&NiagaraSpriteRendererProperties]
  renderers [{rendererClass=&NiagaraSpriteRendererProperties}]
 }
}
summary {
 systemName=NS_Character_PineTrimPack_Head
 emitters [
  {emitterName=NE_Glow, rendererClasses [&NiagaraSpriteRendererProperties]}
  {emitterName=NE_Glow001, rendererClasses [&NiagaraSpriteRendererProperties]}
  {emitterName=NE_Glow002, rendererClasses [&NiagaraSpriteRendererProperties]}
  {emitterName=LowerArms003, rendererClasses [&NiagaraMeshRendererProperties]}
 ]
 userVariables [
  User.bIsFrontEnd {type=NiagaraBool}
  User.bIsFrontEndPreview {type=NiagaraBool}
  User.bisHandHidden {type=NiagaraBool}
  User.DebugReactivity {type=NiagaraBool}
  User.ReactivityCounter {type=NiagaraInt32}
  User.ReactivityDuration {type=NiagaraFloat, defaultValue=4, description=""}
  User.VariantSwitch {type=NiagaraFloat, defaultValue=0, description=""}
 ]
}
