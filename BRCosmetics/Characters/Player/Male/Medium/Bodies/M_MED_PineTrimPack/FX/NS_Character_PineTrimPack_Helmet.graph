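# Niagara cosmetic effect for the PineTrimPack character helmet, spawning short-lived, socket-attached mesh particles that appear to be lightning arcs.
# Spawn rate follows the system timeline, modulated by a stepped loop-age curve for irregular bursts.
# Particles randomize lightning index and Z-mirroring, then drive material parameters and a bright flash that settles and fades.
# User flags zero alpha when the head is hidden and, during ADS, on one of the two mirror variants.
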NiagaraSystem /BRCosmetics/Characters/Player/Male/Medium/Bodies/M_MED_PineTrimPack/FX/NS_Character_PineTrimPack_Helmet
dependencies {
 usedDataInterfaces [&NiagaraDataInterfaceCurve, &NiagaraDataInterfaceArrayInt32, &NiagaraDataInterfaceSocketReader]
 usedDynamicInputs [&InvertBool]
 usedModules [
  &self:SystemSpawnScript.SystemScriptSource.SystemScriptGraph.NiagaraNodeAssignment_1.SetVariables_#1
  &SystemState
  /Game/Effects/Niagara/Modules/Logic/UpdateValueStateChange
  &Timeline
  &self:SystemSpawnScript.SystemScriptSource.SystemScriptGraph.NiagaraNodeAssignment_2.SetVariables_#2
  &self:SystemSpawnScript.SystemScriptSource.SystemScriptGraph.NiagaraNodeAssignment_0.SetVariables_#3
  &self:SystemSpawnScript.SystemScriptSource.SystemScriptGraph.NiagaraNodeAssignment_3.SetVariables_#4
 ]
}
emitters {Helmet {
  emitterUpdateScript [
   &EmitterState(
    Inactive Response=::NewEnumerator1
    Life Cycle Mode=::NewEnumerator1
    Loop Behavior=::NewEnumerator1
    Loop Duration Mode=::NewEnumerator1
    UseLoopDelay=false
   )
   &SpawnRate(
    SpawnRate=&FloatFromCurve(
     FloatCurve=DataInterface() {propertyValues {Curve {keys [
         {tangentMode=RCTM_User, time=0.1, value=0}
         <at 0.3 1>
         {arriveTangent=-0.0405, interpMode=RCIM_Cubic, leaveTangent=-0.0405, tangentMode=RCTM_User, time=0.85, value=1}
         {arriveTangent=-8.0089, interpMode=RCIM_Cubic, leaveTangent=-8.0089, tangentMode=RCTM_User, time=1, value=0}
        ]}}}
     CurveIndex=$System.Timeline.NormalizedTime
     Scale Curve=&Multiply_Float(
      A=10
      B=&FloatFromCurve(
       FloatCurve=DataInterface() {propertyValues {Curve {keys [<at 0 1 RCIM_Constant>,<at 0.2684 0.2 RCIM_Constant>,<at 0.403 2 RCIM_Constant>,<at 0.4998 0 RCIM_Constant>,<at 0.6 1.2824 RCIM_Constant>,<at 1 1 RCIM_Constant>]}}}
       CurveIndex=$System.NormalizedLoopAge
       Scale Curve=$System.SpawnMult
      )
     )
    )
    Use Spawn Probability=false
   )
  ]
  particleSpawnScript [
   &PartitionParticles()
   /SetVariables(Particles.Random:NiagaraInt32=&RandomRangeInt(Minimum=0, Maximum=1)) {isSetParameters=true}
   &InitializeParticle(
    Lifetime=&Multiply_Float(A=&RandomRangeFloat(Minimum=0.4, Maximum=0.65), B=$System.LifetimeMult)
    UsePositionOffset=false
    Write Material Random=true
   ) {module=InitializeParticle001}
   /SetVariables(Particles.LightningIndex=1, Particles.MirrorLightning:NiagaraBool=&RandomBool()) {isSetParameters=true}
   /SetVariables(Particles.LightningIndex:NiagaraInt32=&SelectIntFromArray(
     Direct Array Index=&ReturnParticleIDAttributes()
     Direct Array Mode=::NewEnumerator1
     Int Selection Array=DataInterface() {propertyValues {IntData [0,1,2,3,2,3,1,0,0,2]}}
    )) {isSetParameters=true}
   /SetVariables(Particles.Scale=&MergeHelper_VectorSwitch(0=<1 1 1>, 1=<1 1 -1>, 2=?, 3=?, Index=$Particles.Random, Num=::NewEnumerator0)) {isSetParameters=true}
  ]
  particleUpdateScript [
   &ParticleState()
   &SocketLocation(SamplePosOffset=false, Socket Reader=$System.Socket Reader Interface, Socket Sampling Mode=::NewEnumerator3, VelocitySpeedLimit=false)
   &DynamicMaterialParameters(
    Index 0 Param 1=&Multiply_Float(
     A=&RandomRangeFloat(Minimum=5, Maximum=5)
     B=&FloatFromCurve(FloatCurve=DataInterface() {propertyValues {Curve {keys [{interpMode=RCIM_Cubic, leaveTangent=8.1085, tangentMode=RCTM_Break, time=0.8568, value=1}, {tangentMode=RCTM_User, time=1, value=0}]}}})
    )
    Index 0 Param 2=&Divide_Float(A=&RandomRangeFloat(Minimum=1, Maximum=1.5), B=1.5)
    Index 0 Param 3=&Lerp_Float(
     A=&RandomRangeFloat(Minimum=0, Maximum=0.3)
     B=1
     Alpha=&FloatFromCurve(FloatCurve=DataInterface() {propertyValues {Curve {keys [<at 0 1>,<at 0.1 0.375>,<at 0.5 0.05>,<at 1 0>]}}})
    )
    Index 0 Param 4=&MakeCustomFloatFromBool(Bool=$Particles.InvertLightning, True Float=1, False Float=0)
    Index 1 Param 1=&MakeFloatFromVector(Vector=$Particles.Scale, Channel=::NewEnumerator0)
    Index 1 Param 2=&MakeFloatFromVector(Vector=$Particles.Scale, Channel=::NewEnumerator1)
    Index 1 Param 3=&MakeFloatFromVector(Vector=$Particles.Scale, Channel=::NewEnumerator2)
    Index 2 Param 1=&Multiply_Float(
     A=&FloatFromCurve(FloatCurve=DataInterface() {propertyValues {Curve {keys [<at 0 0.5>,<at 1 1>]}}}, Scale Curve=0.333)
     B=&RandomRangeFloat(Minimum=3, Maximum=4)
    )
    Index 2 Param 2=&FloatFromCurve(
     FloatCurve=DataInterface() {propertyValues {Curve {keys [{interpMode=RCIM_Cubic, tangentMode=RCTM_Break, time=0.1, value=1}, <at 1 0>]}}}
     Scale Curve=&RandomRangeFloat(Minimum=1, Maximum=1)
    )
    Module.Index0Param0Write=true
    Module.Index0Param1Write=true
    Module.Index0Param2Write=true
    Module.Index0Param3Write=true
    Module.Index1Param0Write=true
    Module.Index1Param1Write=true
    Module.Index1Param2Write=true
    Module.Index1Param3Write=true
    Module.Index2Param0Write=true
    Module.Index2Param1Write=true
    Module.Index2Param2Write=true
    Module.Index2Param3Write=true
    Write Parameter Index 1=true
    Write Parameter Index 2=true
   ) {module=DynamicMaterialParameters001}
   &ScaleColor(
    Scale Alpha=&FloatFromCurve(FloatCurve=DataInterface() {propertyValues {Curve {keys [
         {tangentMode=RCTM_User, time=0, value=1}
         {arriveTangent=-0.0405, interpMode=RCIM_Cubic, leaveTangent=-0.0405, tangentMode=RCTM_User, time=0.7009, value=1}
         {arriveTangent=-8.0089, interpMode=RCIM_Cubic, leaveTangent=-8.0089, tangentMode=RCTM_User, time=1, value=0}
        ]}}})
    Scale RGB=&Lerp_Vector(
     A=&VectorFromFloat(Value=10)
     B=&VectorFromFloat(Value=&RandomRangeFloat(Minimum=0.8, Maximum=1, Evaluation Type=::NewEnumerator1))
     Alpha=&FloatFromCurve(FloatCurve=DataInterface() {propertyValues {Curve {keys [
          {interpMode=RCIM_Cubic, leaveTangent=2.5294, tangentMode=RCTM_Break, time=0.1, value=0}
          {arriveTangent=-0.2166, interpMode=RCIM_Cubic, tangentMode=RCTM_Break, time=0.4, value=1}
         ]}}})
    )
    ScaleA=true
    ScaleRGB=true
   ) {module=ScaleColor001}
   /SetVariables(Particles.GlowColor=&MultiplyLinearColor(Linear Color=$StackContext.Color, Scale Factor=<1 1 1 0.2>)) {enabled=false, isSetParameters=true}
   /SetVariables(Particles.GlowColor=&MultiplyLinearColor(
     Linear Color=<rgba 1 1 1 1>
     Scale Factor=&MakeVector4FromVectorAndFloat(
      "Vector (RGB)"=<1 1 1>
      "Float (Alpha)"=&Multiply_Float(
       A=0.2
       B=&FloatFromCurve(
        FloatCurve=DataInterface() {propertyValues {Curve {keys [
            {tangentMode=RCTM_User, time=0, value=1}
            {arriveTangent=-0.0405, interpMode=RCIM_Cubic, leaveTangent=-0.0405, tangentMode=RCTM_User, time=0.7009, value=1}
            {arriveTangent=-8.0089, interpMode=RCIM_Cubic, leaveTangent=-8.0089, tangentMode=RCTM_User, time=1, value=0}
           ]}}}
        Scale Curve=&Multiply_Float(
         A=&MakeCustomFloatFromBool(
          Bool=$User.bIsADS
          True Float=&MergeHelper_FloatSwitch(0=1, 1=0, 2=?, 3=?, Index=$Particles.Random, Num=::NewEnumerator0)
          False Float=1
         )
         B=&MakeCustomFloatFromBool(Bool=$User.bIsHeadHidden, True Float=0, False Float=1)
        )
       )
      )
     )
    )) {isSetParameters=true}
   &ScaleColor(
    Scale Alpha=&MakeCustomFloatFromBool(
     Bool=$User.bIsADS
     True Float=&MergeHelper_FloatSwitch(0=1, 1=0, 2=?, 3=?, Index=$Particles.Random, Num=::NewEnumerator0)
     False Float=1
    )
    ScaleA=true
    ScaleRGB=false
   )
   &ScaleColor(
    Scale Alpha=&MakeCustomFloatFromBool(Bool=$User.bIsHeadHidden, True Float=0, False Float=1)
    ScaleA=true
    ScaleRGB=false
   ) {module=ScaleColor002}
  ]
  rendererClasses [&NiagaraMeshRendererProperties]
  renderers [
   {rendererClass=&NiagaraMeshRendererProperties, rendererIndex=0}
   {rendererClass=&NiagaraMeshRendererProperties, rendererIndex=1}
  ]
 }}
summary {
 systemName=NS_Character_PineTrimPack_Helmet
 emitters [{emitterName=Helmet, rendererClasses [&NiagaraMeshRendererProperties]}]
 userVariables [
  User.bIsADS {type=NiagaraBool}
  User.bIsFrontEnd {type=NiagaraBool}
  User.bIsFrontEndPreview {type=NiagaraBool}
  User.bisHandHidden {type=NiagaraBool}
  User.bIsHeadHidden {type=NiagaraBool}
  User.DebugReactivity {type=NiagaraBool}
  User.ReactivityCounter {type=NiagaraInt32}
  User.ReactivityDuration {type=NiagaraFloat, defaultValue=7, description=""}
  User.SpawnMult {type=NiagaraFloat, defaultValue=1, description=""}
  User.VariantSwitch {type=NiagaraFloat, defaultValue=0, description=""}
 ]
}
