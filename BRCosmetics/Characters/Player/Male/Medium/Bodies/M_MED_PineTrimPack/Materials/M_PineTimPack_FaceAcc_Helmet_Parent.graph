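# Parent character material for the PineTrimPack helmet face accessory, adding a time-triggered glowing symbol effect to the emissive.
# A static switch, PineTrimPack_UseOrangeSkin, inserts the glow between the shared 2023 character PreFX and PostFX functions.
# The glow multiplies a mask texture's red channel, tint and intensity by a timestamp envelope.
# Time since BP_GetTime drives a delayed fade-in and a fade-out after EffectDuration, scaled by a reactivity parameter.
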Material /BRCosmetics/Characters/Player/Male/Medium/Bodies/M_MED_PineTrimPack/Materials/M_PineTimPack_FaceAcc_Helmet_Parent
graph {
 properties {
  bUseMaterialAttributes=true
  bUsedWithMeshDeformer=true
  bUsedWithMorphTargets=true
  bUsedWithSkeletalMesh=true
  pixelDepthOffsetMode=PDOM_Legacy
  shadingModel=MSM_FromMaterialExpression
  usedShadingModels=Default Lit | Subsurface | Clear Coat | Subsurface Profile | Cloth
 }
 nodes [
  n1 = MaterialFunctionCall() {materialFunction=&MF_CharacterParent_2023_PreFX}
  n2 = GetMaterialAttributes(None=@n1.Result) {attributeGetTypes [#1]}
  n3 = Add(A=@n2.Emissive Color, B=NamedRerouteUsage())
  MaterialFunctionCall(
   Material=StaticSwitchParameter(
    True=SetMaterialAttributes(MaterialAttributes=@n2.MaterialAttributes, Emissive Color=@n3) {attributeSetTypes [#1]}
    False=@n1.Result
   ) {defaultValue=true, parameterName=PineTrimPack_UseOrangeSkin}
   M=@n1.M
   Pre Skinned Local Position=@n1.M
   Distance Blend=@n1.M
   Fresnel=@n1.M
  ) {materialFunction=&MF_CharacterParent_2023_PostFX}
  n4 = Subtract(A=Time(), B=ScalarParameter() {defaultValue=9.999999848243207e+30, group=PineTrimPack, parameterName=BP_GetTime, sortPriority=100})
  n5 = Subtract(A=@n4, B=ScalarParameter() {defaultValue=6, parameterName=EffectDuration}) {constB=5}
  n6 = OneMinus(None=Saturate(None=Multiply(A=@n5)))
  n7 = Saturate(None=Multiply(A=Subtract(A=@n4) {constB=0.7}) {constB=0.42})
  NamedRerouteDeclaration(None=Multiply(A=Multiply(A=@n6, B=@n7), B=ScalarParameter() {defaultValue=1, group=PineTrimPack, parameterName=PineTrimPack_UseReactivity})) {name=TimeStamp_GlowingSymbols, nodeColor=<rgba 0 1 0.0625 1>}
  n8 = TextureSample() {
   samplerSource=SSM_Wrap_WorldGroupSettings
   samplerType=SAMPLERTYPE_Masks
   texture=/BRCosmetics/Characters/Player/Male/Medium/Bodies/M_MED_PineTrimPack/Textures/T_PineTrimPack_FaceAcc_Helmet_SL_FX
  }
  n9 = Multiply(
   A=VectorParameter() {group=PineTrimPack, parameterName=TattooGlow_Color, defaultValue=<rgba 1 0.5296 0.1 1>}.RGB
   B=ScalarParameter() {defaultValue=5, group=PineTrimPack, parameterName=TattooGlowIntensity}
  )
  NamedRerouteDeclaration(None=Multiply(A=Multiply(A=@n8.R, B=@n9), B=NamedRerouteUsage())) {name=OutEmissive, nodeColor=<rgba 0 0.7891 1 1>}
 ]
}
