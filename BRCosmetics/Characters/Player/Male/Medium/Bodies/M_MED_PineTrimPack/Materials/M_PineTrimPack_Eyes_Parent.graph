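# Parent eye material for the PineTrimPack character body, adding a reactive colored emissive glow to the shared 2023 eye, eyelash and tech-art override functions.
# Time since BP_GetTime, presumably set by Blueprint, ramps glow from a dim baseline to full after a short delay, then back after EffectDuration.
# The glow is divided by eye adaptation and scaled by a reactivity parameter.
# The UseOrangeSkin static switch adds it to emissive or bypasses it.
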Material /BRCosmetics/Characters/Player/Male/Medium/Bodies/M_MED_PineTrimPack/Materials/M_PineTrimPack_Eyes_Parent
graph {
 properties {
  bUseMaterialAttributes=true
  bUsedWithMeshDeformer=true
  bUsedWithSkeletalMesh=true
  pixelDepthOffsetMode=PDOM_Legacy
  shadingModel=MSM_FromMaterialExpression
  usedShadingModels=Default Lit | Subsurface
 }
 nodes [
  n1 = MaterialFunctionCall(In=MaterialFunctionCall!(Material=MaterialFunctionCall() {materialFunction=/Game/Global/Materials/Character_Parent_2023/Eye_Parent_2023/MF_eyes_2023}.Result) {materialFunction=/Game/Global/Materials/Character_Parent_2023/Eye_Parent_2023/MF_Eyelash}.Result) {materialFunction=&MF_TechArtOverrides}
  n2 = GetMaterialAttributes(None=@n1.Result) {attributeGetTypes [#1]}
  n3 = Multiply(
   A=VectorParameter() {group=PineTrimPack, parameterName=EyeGlow_Color, defaultValue=<rgba 1 0.5271 0.0999 1>}.RGB
   B=ScalarParameter() {defaultValue=10, group=PineTrimPack, parameterName=EyeGlowIntensity}
  )
  n4 = Subtract(A=Time(), B=ScalarParameter() {defaultValue=9.999999848243207e+30, group=PineTrimPack, parameterName=BP_GetTime, sortPriority=100})
  n5 = Multiply(A=Subtract(A=@n4) {constB=0.7}) {constB=0.42}
  n6 = ScalarParameter() {defaultValue=6, parameterName=EffectDuration}
  n7 = Saturate(None=Multiply(A=Subtract(A=@n4, B=@n6) {constB=5}))
  n8 = LinearInterpolate(Alpha=Multiply(A=Saturate(None=@n5), B=OneMinus(None=@n7))) {constA=0.3}
  n9 = Multiply(
   A=Divide(A=Multiply(A=@n3, B=@n8), B=EyeAdaptation().EyeAdaptation)
   B=ScalarParameter() {defaultValue=1, group=PineTrimPack, parameterName=PineTrimPack_UseReactivity}
  )
  StaticSwitchParameter(
   True=SetMaterialAttributes(MaterialAttributes=@n2.MaterialAttributes, Emissive Color=Add(A=@n2.Emissive Color, B=@n9)) {attributeSetTypes [#1]}
   False=@n1.Result
  ) {defaultValue=true, parameterName=PineTrimPack_UseOrangeSkin}
 ]
}
