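# Material function for the PineTrimPack character cosmetic that layers an animated, fiery emissive look with reactive glowing tattoos onto incoming material attributes.
# Panned noise, flicker, fresnel and pre-skinned-position masks build flames, chest glow, lightning and eyebrow glow.
# A BP_GetTime timestamp drives timed elimination and tattoo envelopes, with static switches for tattoo and head variants.
# The emissive-driven result, with zeroed base color, is mask-blended into the attributes.
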MaterialFunction /BRCosmetics/Characters/Player/Male/Medium/Bodies/M_MED_PineTrimPack/Materials/MF_PineTrimPack
graph {nodes [
  CameraVectorWS()
  n1 = FunctionInput() {inputName=Masks, previewValue=<1 0 0 1>}
  ComponentMask(None=@n1) {g=true}
  n2 = GetMaterialAttributes(None=FunctionInput() {inputType=FunctionInput_MaterialAttributes}) {attributeGetTypes [
    #1
    #2
    #3
    #4
   ]}
  n3 = Constant()
  n4 = SetMaterialAttributes!(
   MaterialAttributes=@n2.MaterialAttributes
   EmissiveColor=NamedRerouteUsage()
   BaseColor=NamedRerouteUsage()
   SubsurfaceColor=NamedRerouteUsage()
   Metallic=@n3
   Anisotropy=@n3
   ShadingModel=ShadingModel()
   Specular=ScalarParameter() {defaultValue=0.2, parameterName=PineTrimPack_SpecularOverride}
  ) {attributeSetTypes [
    #1
    #2
    #3
    #5
    #6
    #7
    #8
   ]}
  n5 = Add(A=NamedRerouteUsage(), B=NamedRerouteUsage())
  FunctionOutput(None=BlendMaterialAttributes(A=NamedRerouteUsage(), B=@n4, Alpha=Saturate(None=@n5)))
  Multiply(
   A=GetMaterialAttributes(None=NamedRerouteUsage()) {attributeGetTypes [#2]}.BaseColor
   B=NamedRerouteUsage()
  )
  Multiply()
  NamedRerouteDeclaration(None=@n2.Opacity) {name=InOpacity, nodeColor=<rgba 0.6406 0 1 1>}
  NamedRerouteDeclaration(None=ComponentMask(None=@n1) {b=true}) {name=SkinMask, nodeColor=<rgba 1 0 0.8203 1>}
  n6 = LinearInterpolate(Alpha=NamedRerouteUsage()) {constA=0.75, constB=1.25}
  n7 = TextureObject() {
   samplerType=SAMPLERTYPE_LinearColor
   texture=/BRCosmetics/Characters/Player/Male/Medium/Bodies/M_MED_EggnogFaucet/Materials/T_EggnogFaucet_mask
  }
  n8 = Multiply(A=Time()) {constB=25}
  n9 = VectorParameter() {parameterName=TextureSpeed, defaultValue=<rgba 0 0 -1 1>}
  n10 = NamedRerouteUsage()
  n11 = NamedRerouteUsage()
  n12 = Multiply(A=Subtract(A=Power(Base=@n11)) {constB=0.5}) {constB=2}
  n13 = Add(A=Add(A=Multiply(A=@n8, B=@n9.RGB), B=NamedRerouteUsage()), B=Multiply(A=Multiply(A=@n10, B=@n12)) {constB=2})
  n14 = Multiply(
   A=MaterialFunctionCall(Float2=NamedRerouteUsage()) {materialFunction=&BreakOutFloat2Components}.R
   B=ScalarParameter() {defaultValue=1, parameterName=NormalDistortionAmount}
  )
  n15 = PreSkinnedNormal()
  n16 = ShaderStageSwitch(PixelShader=VertexInterpolator(VS=@n15).PS, VertexShader=@n15)
  n17 = MaterialFunctionCall(Float3=MaterialFunctionCall(
    TextureObject=@n7
    TextureSize=ScalarParameter() {defaultValue=20, parameterName=TextureScale}
    WorldPosition=Add(A=@n13, B=@n14)
    World Space Normal=NamedRerouteDeclaration(None=@n16) {name=PreSkinnedNormal, nodeColor=<rgba 0 0.5078 1 1>}
   ) {materialFunction=&WorldAlignedTexture}.XYZ Texture) {materialFunction=&BreakOutFloat3Components}
  NamedRerouteDeclaration(None=Multiply(A=NamedRerouteUsage(), B=Multiply(A=@n6, B=@n17.G) {constA=0.6})) {name=Noise2, nodeColor=<rgba 1 0.2344 0 1>}
  n18 = ComponentMask(None=NamedRerouteUsage()) {b=true}
  n19 = Time()
  n20 = Sine(None=Add(A=Multiply(A=@n18) {constB=-0.005}, B=Multiply(A=@n19) {constB=0.3}))
  NamedRerouteDeclaration(None=Power(Base=ConstantBiasScale(None=@n20)) {constExponent=5}) {name=Wipe, nodeColor=<rgba 0 1 0.9297 1>}
  n21 = NamedRerouteUsage()
  n22 = Add(A=@n21, B=Multiply(B=NamedRerouteUsage()) {constA=2.5})
  n23 = SphereMask(A=@n22, B=Constant3Vector() {constant=<rgba 0 5 125 0>}) {attenuationRadius=20, hardnessPercent=5}
  n24 = Multiply(A=@n23, B=@n23)
  n25 = Multiply(A=Power(Base=NamedRerouteUsage()) {constExponent=3}, B=@n24)
  n26 = SphereMask(A=@n22, B=Constant3Vector() {constant=<rgba 0 -6 127 0>}) {attenuationRadius=20, hardnessPercent=0}
  n27 = Add(A=Add(A=Multiply(A=@n25) {constB=0.3}, B=Multiply(A=@n24) {constB=0.6}), B=Multiply(A=Multiply(A=@n26, B=@n26)) {constB=0.9})
  NamedRerouteDeclaration(None=Saturate(None=@n27)) {name=ChestGlowMask, nodeColor=<rgba 0 1 0.2266 1>}
  n28 = Constant() {r=0.1}
  n29 = Multiply(A=NamedRerouteDeclaration(None=MaterialFunctionCall!(FlickerInitialSpeed=@n28, FlickerRandom=@n28) {materialFunction=/Game/Effects/Fort_Effects/Materials/_Functions/MF_Flicker}.Result) {name=Flicker, nodeColor=<rgba 0 1 0.3906 1>}) {constB=0.4}
  n30 = ScalarParameter() {defaultValue=1, parameterName=InnerFresnelExponent}
  n31 = Multiply(A=@n30) {constB=1.5}
  n32 = NamedRerouteUsage()
  n33 = Fresnel(ExponentIn=Add(A=@n29, B=LinearInterpolate(A=@n31, B=@n30, Alpha=@n32))) {baseReflectFraction=0, exponent=1.5}
  n34 = LinearInterpolate(Alpha=@n33) {constA=0.8}
  n35 = NamedRerouteUsage()
  n36 = NamedRerouteUsage()
  n37 = Multiply(A=Multiply(A=@n35, B=@n35), B=@n36)
  n38 = LinearInterpolate(Alpha=Power(Base=NamedRerouteUsage()) {constExponent=3}) {constA=0.8, constB=1.2}
  n39 = Add(A=NamedRerouteUsage(), B=Multiply(A=Multiply(A=@n34, B=@n37), B=@n38))
  n40 = Add(A=@n36, B=@n33)
  n41 = Power(Base=OneMinus(None=@n33)) {constExponent=1.5}
  n42 = Multiply(A=Multiply(A=Multiply(A=@n40, B=@n33), B=LinearInterpolate(Alpha=@n41) {constA=0.2}), B=@n38) {constB=1.5}
  n43 = Add(A=@n29, B=ScalarParameter() {defaultValue=3, parameterName=OutterFresnelExponent})
  n44 = Multiply(
   A=@n33
   B=OneMinus(None=MaterialFunctionCall(Exponent=@n43) {materialFunction=&AxisAlignedFresnel}.Fresnel)
  )
  n45 = LinearInterpolate(Alpha=NamedRerouteUsage()) {constA=-0.5}
  n46 = SmoothStep(Value=Multiply(A=@n44, B=Add(A=@n44, B=@n45)))
  n47 = LinearInterpolate(Alpha=Power(Base=NamedRerouteUsage()) {constExponent=3}) {constA=0.1}
  n48 = Add(
   A=Multiply(A=Add(A=@n39, B=@n42), B=ScalarParameter() {defaultValue=1, parameterName=BaseNoiseIntensity})
   B=Multiply(A=Multiply(A=@n46, B=@n47), B=ScalarParameter() {defaultValue=10, parameterName=EdgesNoiseIntensity}) {constB=2}
  )
  n49 = ComponentMask(None=VertexNormalWS()) {b=true}
  n50 = LinearInterpolate(Alpha=Power(Base=ConstantBiasScale(None=@n49)) {constExponent=8}) {constA=1, constB=2}
  n51 = NamedRerouteUsage()
  n52 = Multiply(A=Multiply(A=Multiply(A=@n48, B=@n50), B=LinearInterpolate(Alpha=@n51) {constA=0.8, constB=1.5})) {constB=0.5}
  n53 = Multiply(A=OneMinus(None=NamedRerouteUsage()), B=Power(Base=NamedRerouteUsage()) {constExponent=1.5})
  n54 = ScalarParameter() {defaultValue=1.5, group=PineTrimPack, parameterName=HeadDetailsGlow_Brightness}
  n55 = Add(
   A=Add(A=Add(A=@n27, B=@n52), B=NamedRerouteUsage())
   B=Power(Base=Multiply(A=@n53, B=@n54), Exp=ScalarParameter() {defaultValue=1.5, group=PineTrimPack, parameterName=HeadDetailsGlow_Exp}) {constExponent=1.5}
  )
  NamedRerouteDeclaration(None=LinearInterpolate(
    A=VectorParameter() {parameterName=Color1, defaultValue=<rgba 0.8594 0.0224 0 1>}.RGB
    B=VectorParameter() {parameterName=Color2, defaultValue=<rgba 2 1.5631 0 1>}.RGB
    Alpha=Saturate(None=@n55)
   )) {name=Flames, nodeColor=<rgba 1 0.375 0 1>}
  n56 = SphereMask(A=@n21, B=Constant3Vector() {constant=<rgba 0 5 106 0>}) {attenuationRadius=15, hardnessPercent=0}
  n57 = Multiply(A=@n56, B=@n56)
  n58 = SphereMask(A=@n21, B=Constant3Vector() {constant=<rgba 0 5 105 0>}) {attenuationRadius=4, hardnessPercent=50}
  n59 = Multiply(A=@n58, B=@n58)
  NamedRerouteDeclaration(None=Saturate(None=Subtract(A=@n57, B=@n59))) {name=BellyMask, nodeColor=<rgba 1 0.5156 0 1>}
  NamedRerouteDeclaration(None=@n2.EmissiveColor) {name=InEmissive, nodeColor=<rgba 0 0.1562 1 1>}
  n60 = NamedRerouteUsage()
  n61 = NamedRerouteDeclaration(None=Multiply(A=ComponentMask(None=@n60) {b=true, r=true}) {constB=0.05}) {name=UVsPreskinned, nodeColor=<rgba 0 1 0.1562 1>}
  n62 = ComponentMask(None=VectorParameter() {parameterName=DistorionNoiseSpeed, defaultValue=<rgba 0.1 0.5 0 1>}.RGB) {g=true, r=true}
  n63 = Panner(Coordinate=@n61, Speed=@n62) {speedY=0.5}
  n64 = TextureSample(UVs=BumpOffset(Coordinate=Multiply(A=@n63) {constB=0.5}, Height=Constant() {r=-11})) {
   samplerSource=SSM_Wrap_WorldGroupSettings
   samplerType=SAMPLERTYPE_LinearColor
   texture=&T_EggnogFaucet_mask
  }
  n65 = Time()
  n66 = Sine(None=Add(A=@n64.B, B=Multiply(A=@n65) {constB=0.5}))
  NamedRerouteDeclaration(None=@n66) {name=AnimatedPerlim, nodeColor=<rgba 1 0.6563 0 1>}
  NamedRerouteDeclaration(None=@n64.B) {name=DownGaussianNoise, nodeColor=<rgba 0.3359 0 1 1>}
  n67 = Multiply(A=@n61) {constB=0.1}
  n68 = TextureSample(UVs=BumpOffset(Coordinate=Panner(Coordinate=@n67) {speedX=0.1, speedY=-0.25}, Height=Constant() {r=-7.2})) {
   samplerSource=SSM_Wrap_WorldGroupSettings
   samplerType=SAMPLERTYPE_LinearColor
   texture=&T_EggnogFaucet_mask
  }
  n69 = Time()
  n70 = Sine(None=Add(A=@n68.B, B=Multiply(A=@n69) {constB=0.2}))
  n71 = NamedRerouteUsage()
  NamedRerouteDeclaration(None=StaticSwitchParameter(True=Multiply(A=@n70, B=@n71), False=Constant()) {defaultValue=true, parameterName=PineTrimPack_UseTattoos}) {name=LowFreqGausNoiseAnim, nodeColor=<rgba 0 1 0.0156 1>}
  n72 = Add(A=Multiply(A=@n66) {constB=0.05}, B=@n61)
  n73 = Multiply(A=@n62) {constB=-3.5}
  n74 = TextureSample(UVs=Multiply(A=Panner(Coordinate=@n72, Speed=@n73)) {constB=0.3}) {
   samplerSource=SSM_Wrap_WorldGroupSettings
   samplerType=SAMPLERTYPE_LinearColor
   texture=&T_EggnogFaucet_mask
  }
  NamedRerouteDeclaration(None=@n74.G) {name=MedFreqVoronoi, nodeColor=<rgba 0.6406 0 1 1>}
  n75 = Max(A=NamedRerouteUsage()) {constB=0.25}
  n76 = Multiply(A=NamedRerouteUsage()) {constB=3}
  n77 = NamedRerouteUsage()
  n78 = Time()
  n79 = Constant()
  n80 = Multiply(A=@n78) {constB=-5}
  n81 = Noise!(
   World Position=Add(
    A=NamedRerouteUsage()
    B=MaterialFunctionCall(R=@n78, G=@n79, B=@n80) {materialFunction=&AppendMany}.RGB
   )
   FilterWidth
  ) {levels=4, scale=0.3}
  n82 = Multiply(A=NamedRerouteUsage()) {constB=0.35}
  n83 = Constant2Vector() {g=-1.75, r=0.5}
  n84 = MaterialFunctionCall!(
   UV Coord=Add(A=Multiply(A=@n81) {constB=0.005}, B=Add(A=@n82, B=@n83))
   LenghtScale=ScalarParameter() {defaultValue=1, parameterName=Lightning_PolarLenghtScale}
   RadialScale=ScalarParameter() {defaultValue=0.6, parameterName=Lightining_PolarRadialScale}
  ) {materialFunction=/Game/Global/MaterialFunctions/TechArt/MF_PolarCoordinates}
  n85 = TextureSample(UVs=BumpOffset(Coordinate=Panner(Coordinate=@n84.Result) {speedX=-0.1, speedY=0.01}, Height=Constant() {r=-1})) {
   samplerSource=SSM_Wrap_WorldGroupSettings
   samplerType=SAMPLERTYPE_LinearColor
   texture=/Game/Characters/Player/Male/Medium/Bodies/M_MED_SRDrift/Textures/T_ElectricNoise
  }
  n86 = Multiply(A=@n85.R, B=LinearInterpolate(Alpha=NamedRerouteUsage()) {constA=1.2, constB=0.25}) {constB=2}
  n87 = TextureSample(UVs=BumpOffset(Coordinate=Panner(Coordinate=@n84.Result) {speedX=-0.4, speedY=0.05}, Height=Constant() {r=-0.5})) {
   samplerSource=SSM_Wrap_WorldGroupSettings
   samplerType=SAMPLERTYPE_LinearColor
   texture=&T_ElectricNoise
  }
  n88 = Power(Base=@n87.R)
  n89 = Multiply(A=OneMinus(None=Saturate(None=@n76)), B=Multiply(A=OneMinus(None=@n77), B=Add(A=@n86, B=@n88)))
  NamedRerouteDeclaration(None=Multiply(A=NamedRerouteUsage(), B=Multiply(A=@n75, B=@n89))) {name=Lightning, nodeColor=<rgba 0 0.8125 1 1>}
  n90 = DotProduct(
   A=Transform(None=Constant3Vector() {constant=<rgba 0 1 0 0>}) {transformSourceType=TRANSFORMSOURCE_Local}
   B=MaterialFunctionCall() {materialFunction=&CameraDirectionVector}.Result
  )
  NamedRerouteDeclaration(None=Power(Base=Abs(None=@n90)) {constExponent=6}) {name=FresnelModulator, nodeColor=<rgba 0 1 0.9063 1>}
  n91 = Add(A=NamedRerouteUsage(), B=VectorParameter() {parameterName=RadialNoisePosition, defaultValue=<rgba 0 0 -126.8665 1>}.RGB)
  n92 = MaterialFunctionCall(Vector or UVs=Multiply(A=ComponentMask(None=@n91) {b=true, r=true}) {constB=0.02}) {materialFunction=&VectorToRadialValue}
  n93 = Multiply(A=@n92.Radial Coordinates, B=Constant2Vector() {g=1.5, r=2})
  n94 = Multiply(A=ComponentMask(None=NamedRerouteUsage()) {g=true, r=true}) {constB=0.25}
  n95 = TextureSample(UVs=Panner(Coordinate=Add(A=@n93, B=@n94)) {speedX=0.05, speedY=-1}) {
   samplerType=SAMPLERTYPE_LinearColor
   texture=&T_EggnogFaucet_mask
  }
  NamedRerouteDeclaration(None=Multiply(A=@n95.G, B=@n92.Linear Distance)) {name=RadialVoronoi, nodeColor=<rgba 0.8984 0 1 1>}
  n96 = ComponentMask(None=@n1) {r=true}
  n97 = ComponentMask(None=NamedRerouteUsage()) {b=true}
  n98 = Saturate(None=Multiply(A=Add(A=@n97) {constB=-140}) {constB=0.1})
  n99 = Power(
   Base=@n96
   Exp=LinearInterpolate(
    A=ScalarParameter() {defaultValue=2, group=PineTrimPack, parameterName=AmbientOcclusionPower}
    Alpha=NamedRerouteDeclaration(None=@n98) {name=HeadMask, nodeColor=<rgba 1 0 0.4687 1>}
   ) {constB=0.35}
  )
  n100 = Multiply(A=NamedRerouteUsage(), B=ScalarParameter() {defaultValue=2, parameterName=BellyMaskIntensity}) {constB=0}
  NamedRerouteDeclaration(None=Saturate(None=Add(A=@n99, B=@n100))) {name=AmbientOcclusionMask, nodeColor=<rgba 0 0.6953 1 1>}
  n101 = TextureSample(UVs=BumpOffset(Coordinate=Multiply(A=@n61) {constB=0.5}, Height=Constant() {r=-4})) {
   samplerSource=SSM_Wrap_WorldGroupSettings
   samplerType=SAMPLERTYPE_LinearColor
   texture=&T_EggnogFaucet_mask
  }
  NamedRerouteDeclaration(None=StaticSwitchParameter(True=@n101.B, False=Constant()) {defaultValue=true, parameterName=PineTrimPack_UseTattoos}) {name=StaticPerlin, nodeColor=<rgba 0.4297 1 0 1>}
  n102 = NamedRerouteUsage()
  n103 = Multiply(A=LinearInterpolate(B=Constant3Vector() {constant=<rgba 1 0.1 0.004 1>}, Alpha=@n102) {constA=1}, B=NamedRerouteUsage())
  n104 = Constant()
  n105 = NamedRerouteUsage()
  n106 = LinearInterpolate(
   A=NamedRerouteUsage()
   B=Multiply(A=MaterialFunctionCall(Color_Morning=@n104, Color_Day=@n104, Color_Evening=@n105, Color_Night=@n105) {materialFunction=&Color_Time_of_Day}.Result) {constB=0.1}
  ) {constAlpha=0.25, constB=0}
  n107 = Multiply(
   A=LinearInterpolate(Alpha=@n102) {constA=1, constB=2.5}
   B=LinearInterpolate(A=VectorParameter() {parameterName=ShadowTint, defaultValue=<rgba 0.401 0.056 0 1>}.RGB, B=Add(A=@n103, B=@n106), Alpha=NamedRerouteUsage())
  )
  n108 = Multiply(
   A=VectorParameter() {group=PineTrimPack, parameterName=TattooGlow_Color, defaultValue=<rgba 1 0.5296 0.1 1>}.RGB
   B=ScalarParameter() {defaultValue=5, group=PineTrimPack, parameterName=TattooGlowIntensity}
  )
  n109 = TextureSampleParameter2D() {
   group=PineTrimPack
   parameterName=ReactiveTattooMask
   samplerSource=SSM_Wrap_WorldGroupSettings
   samplerType=SAMPLERTYPE_LinearColor
   texture=/BRCosmetics/Characters/Player/Male/Medium/Bodies/M_MED_PineTrimPack/Textures/T_PineTrimPack_Body_SL_FX
  }
  n110 = LinearInterpolate(Alpha=NamedRerouteUsage()) {constA=0.05}
  n111 = StaticSwitchParameter(
   True=LinearInterpolate(A=@n107, B=NamedRerouteDeclaration(None=@n108) {name=TattooColor, nodeColor=<rgba 1 0.6328 0 1>}, Alpha=Multiply(A=@n109.R, B=@n110))
   False=@n107
  ) {defaultValue=true, parameterName=PineTrimPack_UseTattoos}
  n112 = LinearInterpolate(A=@n111, B=NamedRerouteUsage(), Alpha=Multiply(A=NamedRerouteUsage(), B=NamedRerouteUsage()))
  n113 = NamedRerouteDeclaration(None=FunctionInput() {bUsePreviewValueAsDefault=true, inputName=Fresnel, inputType=FunctionInput_Scalar}) {name=InFresnel, nodeColor=<rgba 0.6641 1 0 1>}
  n114 = Add(
   A=StaticSwitchParameter(True=StaticSwitchParameter(True=@n112, False=@n111) {parameterName=PineTrimPack_IsHead}, False=@n112) {defaultValue=true, parameterName=PineTrimPack_UseTattoos}
   B=Multiply(A=Power(Base=@n113) {constExponent=3}, B=Constant3Vector() {constant=<rgba 0.3524 0.1865 0 1>})
  )
  n115 = Power(Base=OneMinus(None=NamedRerouteUsage())) {constExponent=6}
  n116 = Multiply(A=NamedRerouteUsage(), B=Constant2Vector() {g=1, r=0.5}) {constB=0.5}
  n117 = TextureSample(UVs=BumpOffset(Coordinate=Panner(Coordinate=@n116) {speedX=0.06, speedY=-0.5}, Height=Constant() {r=-5})) {
   samplerSource=SSM_Wrap_WorldGroupSettings
   samplerType=SAMPLERTYPE_LinearColor
   texture=&T_EggnogFaucet_mask
  }
  n118 = Multiply(A=Multiply(A=@n117.G, B=Constant3Vector() {constant=<rgba 0.1892 0.0821 0.0152 1>}), B=ScalarParameter() {parameterName=BumpNoiseIntensity}) {constB=0.1}
  n119 = StaticSwitchParameter(True=Multiply(A=NamedRerouteUsage(), B=Multiply(A=@n115, B=@n118)), False=Constant()) {defaultValue=true, parameterName=PineTrimPack_UseTattoos}
  NamedRerouteDeclaration(None=Multiply(A=ScalarParameter() {defaultValue=0.55, group=PineTrimPack, parameterName=PineTrimPack_EmissiveBrightness}, B=Add(A=@n114, B=@n119))) {name=OutEmissive, nodeColor=<rgba 0.4766 1 0 1>}
  n120 = Constant() {r=0.1}
  n121 = MaterialFunctionCall(FlickerInitialSpeed=@n120, FlickerRandom=@n120) {materialFunction=&MF_Flicker}
  n122 = NamedRerouteUsage()
  n123 = Constant() {r=4}
  n124 = Multiply(
   A=LinearInterpolate(Alpha=@n121.Result) {constA=1, constB=2}
   B=Power(Base=OneMinus(None=@n122), Exp=Subtract(A=@n123, B=@n121.Result)) {constExponent=6}
  )
  n125 = NamedRerouteUsage()
  n126 = Add(A=@n124, B=Multiply(A=@n125, B=@n125))
  n127 = Multiply(A=Multiply(A=Multiply(A=@n126, B=@n124)), B=Power(Base=NamedRerouteUsage()))
  n128 = Multiply(A=Power(Base=@n125))
  NamedRerouteDeclaration(None=Saturate(None=Multiply(A=@n127, B=@n128))) {name=DetailFlames, nodeColor=<rgba 1 0 0.75 1>}
  NamedRerouteDeclaration(None=@n2.MaterialAttributes) {name=InMA, nodeColor=<rgba 0 1 0.8594 1>}
  n129 = ScalarParameter() {defaultValue=1, group=PineTrimPack, parameterName=PineTrimPack_UseReactivity}
  n130 = Subtract(A=Time(), B=ScalarParameter() {defaultValue=9.999999848243207e+30, group=PineTrimPack, parameterName=BP_GetTime, sortPriority=100})
  n131 = Subtract(A=@n130, B=ScalarParameter() {defaultValue=6, parameterName=EffectDuration}) {constB=5}
  n132 = OneMinus(None=Saturate(None=Multiply(A=@n131)))
  n133 = Saturate(None=Multiply(A=Subtract(A=@n130) {constB=0.7}) {constB=0.42})
  NamedRerouteDeclaration(None=Multiply(A=@n129, B=Multiply(A=@n132, B=@n133))) {name=TimeStamp_GlowingTattoos, nodeColor=<rgba 0 1 0.0625 1>}
  n134 = Saturate(None=Multiply(A=@n130))
  NamedRerouteDeclaration(None=Multiply(A=Multiply(A=@n134, B=@n132), B=@n129)) {name=TimeStamp_Eliminations, nodeColor=<rgba 0.2891 0 1 1>}
  n135 = TextureSample() {texture=/BRCosmetics/Characters/Player/Male/Medium/Bodies/M_MED_PineTrimPack/Textures/T_PineTrimPack_AdditionalHairMask_FaceAcc_FX}
  n136 = Add(A=OneMinus(None=ChannelMaskParameter(None=@n135.RGB) {parameterName=UVSeams_Channel}.R)) {constB=0.35}
  NamedRerouteDeclaration(None=StaticSwitchParameter(True=Saturate(None=@n136), False=Constant() {r=1}) {defaultValue=true, parameterName=PineTrimPack_UseTattoos}) {name=SeamsMask, nodeColor=<rgba 0 0.4609 1 1>}
  n137 = OneMinus(None=NamedRerouteUsage())
  n138 = NamedRerouteUsage()
  n139 = Saturate(None=Add(A=ComponentMask(None=@n138) {b=true}) {constB=-156})
  n140 = Constant()
  n141 = NamedRerouteUsage()
  n142 = NamedRerouteUsage()
  n143 = Normalize(VectorInput=Constant3Vector() {constant=<rgba 0 -1 -2 0>})
  n144 = Normalize(VectorInput=Constant3Vector() {constant=<rgba 0 0 1 0>})
  n145 = Add(A=Saturate(None=DotProduct(A=@n142, B=@n143)), B=Saturate(None=DotProduct(A=@n142, B=@n144)))
  n146 = Multiply(A=Multiply(A=Power(Base=@n141) {constExponent=6}) {constB=0.9}, B=OneMinus(None=Saturate(None=@n145)))
  n147 = Saturate(None=Add(A=Multiply(A=@n137, B=@n139), B=StaticSwitchParameter(True=@n140, False=@n146) {defaultValue=true, parameterName=PineTrimPack_UseTattoos}))
  n148 = VectorParameter() {parameterName=EyebrowsColors_Emissive, defaultValue=<rgba 1 0.3133 0.084 1>}
  NamedRerouteDeclaration(None=Multiply(A=NamedRerouteDeclaration(None=@n147) {name=EyebrowsMask, nodeColor=<rgba 1 0 0.3281 1>}, B=Multiply(A=@n148.RGB) {constB=7})) {name=EyebrowsEmissive, nodeColor=<rgba 0.3125 0 1 1>}
  NamedRerouteDeclaration(None=@n96) {name=HairStrandsMask, nodeColor=<rgba 1 0.5156 0 1>}
  NamedRerouteDeclaration(None=FunctionInput() {inputName=PreSkinnedPosition}) {name=PreSkinnedPosition, nodeColor=<rgba 1 0.9141 0 1>}
  n149 = Max(A=NamedRerouteUsage()) {constB=0.8}
  n150 = Time()
  n151 = Sine(None=Add(A=NamedRerouteDeclaration(None=@n74.B) {name=MedFreqNoise, nodeColor=<rgba 0.3125 1 0 1>}, B=Multiply(A=@n150) {constB=0.5}))
  n152 = NamedRerouteUsage()
  n153 = StaticSwitchParameter(True=Multiply(A=ConstantBiasScale(None=@n151), B=Power(Base=@n152)), False=Constant()) {defaultValue=true, parameterName=PineTrimPack_UseTattoos}
  NamedRerouteDeclaration(None=Multiply(A=Multiply(A=@n149, B=@n153), B=NamedRerouteUsage())) {name=MedFreqNoiseAnim, nodeColor=<rgba 1 0 0.7266 1>}
  NamedRerouteDeclaration(None=@n2.BaseColor) {name=InBaseColor, nodeColor=<rgba 1 0.0469 0 1>}
  NamedRerouteDeclaration(None=@n2.SubsurfaceColor) {name=InSSSColor, nodeColor=<rgba 0 0.2734 1 1>}
  NamedRerouteDeclaration(None=Constant()) {name=OutBaseColor, nodeColor=<rgba 0.5938 1 0 1>}
  NamedRerouteDeclaration(None=Constant()) {name=OutSSSColor, nodeColor=<rgba 0 1 0.6719 1>}
 ]}
