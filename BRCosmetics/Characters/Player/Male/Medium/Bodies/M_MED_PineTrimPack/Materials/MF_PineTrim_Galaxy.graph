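# Material function that adds an animated galaxy glow to the PineTrim character cosmetic's emissive, excluded by a hair/face-accessory mask.
# A camera-facing virtual plane anchored at the head or an override position, compensated for distance and FOV, supplies screen-aligned UVs.
# Galaxy layers combine scaled, polar-panned and rotating texture samples with electric-noise orb rays, which use a fresnel flowmap and cheaper low/medium quality paths.
# A Blueprint timestamp drives a timed reactivity fade.
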MaterialFunction /BRCosmetics/Characters/Player/Male/Medium/Bodies/M_MED_PineTrimPack/Materials/MF_PineTrim_Galaxy
graph {nodes [
  Constant2Vector() {g=0.74, r=5}
  Constant2Vector() {g=1, r=7}
  Constant() {r=200}
  n1 = GetMaterialAttributes(None=FunctionInput() {inputType=FunctionInput_MaterialAttributes}) {attributeGetTypes [#1]}
  n2 = NamedRerouteUsage()
  n3 = NamedRerouteUsage()
  n4 = Divide(B=ComponentMask(None=TransformPosition(None=@n3) {transformSourceType=TRANSFORMPOSSOURCE_World, transformType=TRANSFORMPOSSOURCE_View}) {b=true})
  n5 = TextureSample(UVs=MaterialFunctionCall(UVs=@n2, Texture Scale=Multiply(A=@n4) {constB=6}) {materialFunction=&ScaleUVsByCenter}.UVs) {
   samplerSource=SSM_Wrap_WorldGroupSettings
   samplerType=SAMPLERTYPE_LinearColor
   texture=/BRCosmetics/Characters/Player/Male/Medium/Bodies/M_MED_PineTrimPack/Textures/T_PineTrimPack_GalaxyBrain
  }
  n6 = Constant3Vector() {constant=<rgba 1 0.7287 0.4045 1>}
  n7 = MaterialFunctionCall!(
   UV Coord=MaterialFunctionCall(UVs=@n2, Texture Scale=Multiply(A=@n4) {constB=1.5}) {materialFunction=&ScaleUVsByCenter}.UVs
   LenghtScale=Constant() {r=0.1}
   RadialScale
  ) {materialFunction=/Game/Global/MaterialFunctions/TechArt/MF_PolarCoordinates}
  n8 = TextureSample(UVs=Panner(Coordinate=@n7.Result) {speedX=-0.075, speedY=0.1}) {
   samplerSource=SSM_Wrap_WorldGroupSettings
   samplerType=SAMPLERTYPE_LinearColor
   texture=&T_PineTrimPack_GalaxyBrain
  }
  n9 = Multiply(A=Power(Base=@n8.B) {constExponent=4}, B=Multiply(A=Constant3Vector() {constant=<rgba 1 0.1848 0 1>}) {constB=0.5})
  n10 = Add(A=Multiply(A=@n8.B) {constB=2}) {constB=-1}
  n11 = Rotator(Coordinate=@n2, Time=Multiply(A=Time()))
  n12 = Multiply(A=@n4, B=ScalarParameter() {defaultValue=11, group=SUPERLEVEL, parameterName=GalaxyCenter_Scale, sortPriority=101}) {constB=8}
  n13 = TextureSample(UVs=Add(
    A=Multiply(A=@n10) {constB=0.04}
    B=MaterialFunctionCall(UVs=@n11, Texture Scale=@n12) {materialFunction=&ScaleUVsByCenter}.UVs
   )) {
   samplerSource=SSM_Clamp_WorldGroupSettings
   samplerType=SAMPLERTYPE_LinearColor
   texture=&T_PineTrimPack_GalaxyBrain
  }
  n14 = Multiply(A=Constant3Vector() {constant=<rgba 1 0.9176 0.4219 1>}) {constB=0.5}
  n15 = Add(A=Multiply(A=@n5.A, B=Multiply(A=@n6) {constB=0.6}), B=Add(A=Multiply(A=@n9, B=@n13.R), B=Multiply(A=@n13.G, B=@n14)))
  n16 = NamedRerouteUsage()
  n17 = Multiply(A=NamedRerouteUsage(), B=Saturate(None=Multiply(A=@n16) {constB=1.75}))
  n18 = ScalarParameter() {defaultValue=0.75, parameterName=GalaxyBrain_Intensity}
  n19 = NamedRerouteUsage()
  n20 = Add(A=@n1.EmissiveColor, B=Multiply(A=Add(A=@n15, B=@n17), B=LinearInterpolate(A=@n18, Alpha=@n19) {constB=2}) {constB=0.5})
  n21 = TextureSample() {
   samplerSource=SSM_Wrap_WorldGroupSettings
   texture=/BRCosmetics/Characters/Player/Male/Medium/Bodies/M_MED_PineTrimPack/Textures/T_PineTrimPack_AdditionalHairMask_FaceAcc_FX
  }
  n22 = Power(Base=@n21.G)
  FunctionOutput(None=SetMaterialAttributes!(MaterialAttributes=@n1.MaterialAttributes, EmissiveColor=LinearInterpolate(A=@n20, B=@n1.EmissiveColor, Alpha=@n22)) {attributeSetTypes [#1]})
  NamedRerouteDeclaration(None=@n7.Result) {name=PolarCoordinates, nodeColor=<rgba 0 1 0.5078 1>}
  NamedRerouteDeclaration(None=@n13.R) {name=Center Glow Mask, nodeColor=<rgba 1 0.8203 0 1>}
  n23 = NamedRerouteUsage()
  n24 = TransformPosition(None=NamedRerouteUsage()) {transformSourceType=TRANSFORMPOSSOURCE_World, transformType=TRANSFORMPOSSOURCE_View}
  n25 = Multiply(A=Divide(B=ComponentMask(None=@n24) {b=true})) {constB=100}
  n26 = NamedRerouteDeclaration(None=MaterialFunctionCall(
    UV Coord=MaterialFunctionCall(UVs=@n23, Texture Scale=@n25) {materialFunction=&ScaleUVsByCenter}.UVs
    LenghtScale=ScalarParameter() {defaultValue=20, parameterName=OrbRays_LengthScale}
    RadialScale=ScalarParameter() {defaultValue=2, parameterName=OrbRays_RadialScale}
   ) {materialFunction=&MF_PolarCoordinates}.Result) {name=RadialCoords, nodeColor=<rgba 0.7578 0 1 1>}
  n27 = VectorParameter() {group=RelayStick_Orb, parameterName="OrbRadialCoordScales (UVUV)", defaultValue=<rgba 0.1366 2 0.174 1>}
  n28 = AppendVector(A=@n27.R, B=@n27.R)
  n29 = Add(A=Time(), B=ScalarParameter() {defaultValue=0.1, group=RelayStick_Orb, parameterName=Color Lightning Audio Jitter})
  n30 = VectorParameter() {group=RelayStick_Orb, parameterName="OrbRadialCoordPanSpeeds (UVUV)", defaultValue=<rgba 0.0333 0.0333 -0.0267 -0.0233>}
  n31 = TextureCoordinate() {uTiling=3, vTiling=3}
  n32 = DDX(Value=@n31)
  n33 = DDY(Value=@n31)
  n34 = TextureSample!(
   UVs=Panner(Coordinate=Multiply(A=@n26, B=@n28) {constB=3}, Time=@n29, Speed=AppendVector(A=@n30.R, B=@n30.R)) {speedX=-1.1, speedY=-0.067}
   Tex
   "DDX(UVs)"=@n32
   "DDY(UVs)"=@n33
   Apply View MipBias
  ) {
   mipValueMode=TMVM_Derivative
   samplerSource=SSM_Wrap_WorldGroupSettings
   samplerType=SAMPLERTYPE_LinearColor
   texture=/Game/Characters/Player/Male/Medium/Bodies/M_MED_SRDrift/Textures/T_ElectricNoise
  }
  n35 = AppendVector(A=@n27.B, B=@n27.B)
  n36 = TextureSample(
   UVs=Panner(Coordinate=Multiply(A=@n26, B=@n35), Time=@n29, Speed=AppendVector(A=@n30.B, B=@n30.B)) {speedX=-0.33, speedY=-0.053}
   "DDX(UVs)"=@n32
   "DDY(UVs)"=@n33
  ) {
   mipValueMode=TMVM_Derivative
   samplerSource=SSM_Wrap_WorldGroupSettings
   samplerType=SAMPLERTYPE_LinearColor
   texture=&T_ElectricNoise
  }
  n37 = MaterialFunctionCall!(Blend=NamedRerouteUsage(), Base=QualitySwitch(Default=Add(A=@n34.R, B=@n36.R), Low=@n34.R, Medium=@n34.R)) {materialFunction=/Engine/Functions/Engine_MaterialFunctions03/Blends/Blend_Screen}
  n38 = ScalarParameter() {defaultValue=3, group=RelayStick_Orb, parameterName=Blended Texture Exponent}
  n39 = ScalarParameter() {defaultValue=3, group=RelayStick_Orb, parameterName=Lightning Edge Fade Exponent}
  n40 = Abs(None=DotProduct(A=PixelNormalWS(), B=CameraVectorWS()))
  n41 = SquareRoot(None=@n40)
  n42 = ScalarParameter() {defaultValue=1, group=RelayStick_Orb, parameterName=Fresnel Flowmap Intensity Max}
  n43 = ScalarParameter() {defaultValue=0.1, group=RelayStick_Orb, parameterName=Fresnel Flowmap Intensity Min}
  n44 = AppendVector(A=Multiply(A=AppendVector(A=@n41, B=@n41), B=LinearInterpolate(A=@n42, B=@n43, Alpha=@n40)), B=Constant())
  n45 = Normalize(VectorInput=Add(A=Transform(None=@n44) {transformSourceType=TRANSFORMSOURCE_View}, B=PixelNormalWS()))
  n46 = VertexNormalWS()
  n47 = Abs(None=DotProduct(A=QualitySwitch(Default=@n45, Low=@n46, Medium=@n46) {desc=Skipping flowmap on med/low}, B=CameraVectorWS()))
  n48 = LinearInterpolate(
   A=VectorParameter() {group=RelayStick_Orb, parameterName=Color Lightning Falloff, defaultValue=<rgba 0.9 0.2959 0 3>}.RGB
   B=VectorParameter() {group=RelayStick_Orb, parameterName=Color Lightning, defaultValue=<rgba 1 0.4999 0 1>}.RGB
   Alpha=Multiply(A=@n37.Result, B=Power(Base=@n47, Exp=@n39) {constExponent=3})
  )
  n49 = ScalarParameter() {defaultValue=1, group=RelayStick_Orb, parameterName=Color Lightning Intensity}
  NamedRerouteDeclaration(None=Multiply(A=Power(Base=@n37.Result, Exp=@n38), B=Multiply(A=@n48, B=@n49))) {name=Orb Rays, nodeColor=<rgba 0.0781 0 1 1>}
  n50 = Subtract(A=Time(), B=ScalarParameter() {defaultValue=9.999999848243207e+30, group=PineTrimPack, parameterName=BP_GetTime, sortPriority=100})
  n51 = Subtract(A=@n50, B=ScalarParameter() {defaultValue=6, parameterName=EffectDuration}) {constB=5}
  n52 = OneMinus(None=Saturate(None=Multiply(A=@n51)))
  n53 = Saturate(None=Multiply(A=Subtract(A=@n50) {constB=0.7}) {constB=0.42})
  NamedRerouteDeclaration(None=Multiply(A=ScalarParameter() {defaultValue=1, group=PineTrimPack, parameterName=PineTrimPack_UseReactivity}, B=Multiply(A=@n52, B=@n53))) {name=TimeStamp_GlowingTattoos, nodeColor=<rgba 0 1 0.0625 1>}
  n54 = MaterialFunctionCall() {materialFunction=&CameraDirectionVector}
  n55 = Multiply(A=Constant3Vector() {constant=<rgba 1 1 0 0>}, B=@n54.Result)
  n56 = Add(A=Multiply(B=Normalize(VectorInput=@n55)) {constA=0.25}, B=@n54.Result)
  n57 = VectorParameter() {group=Superlevel, parameterName=CorePosition_Override, sortPriority=600000, defaultValue=<rgba 0 0 125 0>}
  n58 = VectorParameter() {group=Superlevel, parameterName=LocalHeadPosition, sortPriority=1000, defaultValue=<rgba 0 0 160 0>}
  n59 = TransformPosition(None=Add(A=@n58.RGB, B=VectorParameter() {group=Superlevel, parameterName=PreviewOffset_Head, sortPriority=800}.RGB)) {transformType=TRANSFORMPOSSOURCE_World}
  n60 = VectorParameter() {group=Superlevel, parameterName=BP_OffsetVector_Head, sortPriority=1000}
  n61 = Add(
   A=Multiply(
    A=ScalarParameter() {defaultValue=8.5, group=SUPERLEVEL, parameterName=PositionOffset_Head, sortPriority=150}
    B=Normalize(VectorInput=@n60.RGB)
   )
   B=VectorParameter() {group=Superlevel, parameterName=BP_HeadPosition, sortPriority=1000}.RGB
  )
  n62 = ScalarParameter() {group=SUPERLEVEL, parameterName=BP_UseCCPM, sortPriority=0}
  n63 = NamedRerouteDeclaration(None=StaticSwitchParameter(True=TransformPosition(None=@n57.RGB) {transformType=TRANSFORMPOSSOURCE_World}, False=LinearInterpolate(A=@n59, B=@n61, Alpha=@n62)) {group=SUPERLEVEL, parameterName=OverrideCorePosition, sortPriority=600000}) {name=HeadPosition, nodeColor=<rgba 0 1 0.2734 1>}
  n64 = Constant() {r=45}
  n65 = Constant() {r=90}
  n66 = MaterialFunctionCall() {materialFunction=/Engine/Functions/Engine_MaterialFunctions02/Utility/FOV}
  n67 = Multiply(
   A=LinearInterpolate(A=Constant() {r=64}, B=Constant() {r=128}, Alpha=InverseLinearInterpolate(A=@n64, B=@n65, Value=@n66.FOV angle in degrees))
   B=Distance(A=@n63, B=CameraPositionWS())
  )
  NamedRerouteDeclaration(None=MaterialFunctionCall(Plane Normal Axis=Normalize(VectorInput=@n56), Plane Center=Add(A=@n54.Result, B=@n63), UV Scale=Multiply(A=@n67) {constB=0.0146}) {materialFunction=&VirtualPlaneCoordinates}.UVs) {name=ChestAllignedScreenPosition, nodeColor=<rgba 0.125 1 0 1>}
 ]}
