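# Parent hair material for the PineTrimPack character skins, built between the 2023 character PreFX and PostFX functions.
# Masked Fortnite Hair 2023 shading is blended in, with an optional flyaway hair pass.
# The default-on orange-skin switch adds SimpleSlurp emissive, then chains PineTrimPack and Galaxy functions.
# The Sift switch replaces emissive with a face-accessory glow texture, beside a timestamp-driven fade-in/out envelope for reactive glowing symbols.
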Material /BRCosmetics/Characters/Player/Male/Medium/Bodies/M_MED_PineTrimPack/Materials/M_PineTrimPack_Hair_Parent
graph {
 properties {
  bUseMaterialAttributes=true
  bUsedWithInstancedStaticMeshes=true
  bUsedWithMeshDeformer=true
  bUsedWithMorphTargets=true
  bUsedWithSkeletalMesh=true
  pixelDepthOffsetMode=PDOM_Legacy
  shadingModel=MSM_FromMaterialExpression
  usedShadingModels=Default Lit | Subsurface | Clear Coat | Subsurface Profile | Cloth
 }
 nodes [
  n1 = MaterialFunctionCall() {materialFunction=&MF_CharacterParent_2023_PreFX}
  n2 = MaterialFunctionCall(Color=@n1.M) {materialFunction=&SplitComponents}
  n3 = MaterialFunctionCall(R=@n2.R, G=@n2.R, B=Constant() {r=1}) {materialFunction=&AppendMany}
  n4 = MaterialFunctionCall!(In=@n1.Result, Mask Input=@n1.Result) {materialFunction=/Game/Global/Materials/Character_Parent_2023/Hair_Parent_2023/MaterialFunctions/MF_FN_Hair_2023}
  n5 = BlendMaterialAttributes(
   A=@n1.Result
   B=StaticSwitchParameter(True=@n4.Result, False=@n1.Result) {defaultValue=true, group=Skin FX, parameterName=UseFNHair2_0}
   Alpha=@n4.HairMask
  )
  n6 = StaticSwitchParameter(
   True=MaterialFunctionCall!(MaterialInput=@n5) {materialFunction=/Game/Global/Materials/Character_Parent_2023/Hair_Parent_2023/MaterialFunctions/MF_FlyAwayHair_2023}.Result
   False=@n5
  ) {group=Skin FX, parameterName=IsFlyAwayHair?}
  n7 = SetMaterialAttributes(
   MaterialAttributes=@n6
   Emissive Color=MaterialFunctionCall() {materialFunction=/Game/Global/Materials/MaterialGym/Materials/MaterialElements/Depth/SimpleSlurp/MF_SimpleSlurp}.Result
  ) {attributeSetTypes [#1]}
  n8 = StaticSwitchParameter(
   True=MaterialFunctionCall!(In=MaterialFunctionCall!(PreSkinnedPosition=@n1.Pre skinned Local Position, Masks=@n3.RGB, Fresnel, In=@n7) {materialFunction=&/MF_PineTrimPack}.Result) {materialFunction=&/MF_PineTrim_Galaxy}.Result
   False=@n6
  ) {defaultValue=true, parameterName=PineTrimPack_UseOrangeSkin}
  n9 = TextureSample() {
   samplerSource=SSM_Wrap_WorldGroupSettings
   texture=/BRCosmetics/Characters/Player/Male/Medium/Bodies/M_MED_PineTrimPack_Sift/Textures/T_PineTrimPack_Sift_FaceAcc_E
  }
  n10 = Multiply(A=Multiply(A=Power(Base=@n9.RGB) {constExponent=1.5}) {constB=7}, B=NamedRerouteUsage())
  MaterialFunctionCall(
   Material=StaticSwitchParameter(
    True=SetMaterialAttributes(MaterialAttributes=@n8, Emissive Color=@n10) {attributeSetTypes [#1]}
    False=@n8
   ) {parameterName=PineTrimPack_UseSiftSkin}
   M=@n1.M
   Pre Skinned Local Position=@n1.M
   Distance Blend=@n1.M
   Fresnel=@n1.M
  ) {materialFunction=&MF_CharacterParent_2023_PostFX}
  n11 = Subtract(A=Time(), B=ScalarParameter() {defaultValue=9.999999848243207e+30, group=PineTrimPack, parameterName=BP_GetTime, sortPriority=100})
  n12 = Subtract(A=@n11, B=ScalarParameter() {defaultValue=6, parameterName=EffectDuration}) {constB=5}
  n13 = OneMinus(None=Saturate(None=Multiply(A=@n12)))
  n14 = Saturate(None=Multiply(A=Subtract(A=@n11) {constB=0.3}) {constB=1.5})
  NamedRerouteDeclaration(None=Multiply(A=Multiply(A=@n13, B=@n14), B=ScalarParameter() {defaultValue=1, group=PineTrimPack, parameterName=PineTrimPack_UseReactivity})) {name=TimeStamp_GlowingSymbols, nodeColor=<rgba 0 1 0.0625 1>}
  Reroute(None=@n1.Tech Art Mask)
 ]
}
