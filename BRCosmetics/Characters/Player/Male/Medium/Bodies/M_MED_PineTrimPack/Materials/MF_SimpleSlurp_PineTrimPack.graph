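# Material function for the PineTrimPack character body that outputs a "SimpleSlurp" color, blending two tunable, multiplied colors.
# A squared sphere mask on pre-skinned position swaps in a separate belly tint for the second color.
# The blend mask appears to be the saturated product of a double fresnel rim band, a world-normal directional gradient and fake sun scattering from the atmospheric light and camera vectors.
# Named reroutes link the stages.
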MaterialFunction /BRCosmetics/Characters/Player/Male/Medium/Bodies/M_MED_PineTrimPack/Materials/MF_SimpleSlurp_PineTrimPack
graph {nodes [
  FunctionOutput(None=NamedRerouteUsage())
  n1 = VectorParameter() {group=SimpleSlurp, parameterName=SS_Color02, sortPriority=2, defaultValue=<rgba 0.0089 0.0153 0.0469 1>}
  n2 = VectorParameter() {group=SimpleSlurp, parameterName=SS_Color02_Belly, sortPriority=2, defaultValue=<rgba 0.0089 0.0153 0.0469 1>}
  n3 = LinearInterpolate(A=@n1.RGB, B=@n2.RGB, Alpha=NamedRerouteUsage())
  n4 = ScalarParameter() {defaultValue=1, group=SimpleSlurp, parameterName=SS_Color02Multiplier, sortPriority=3}
  n5 = VectorParameter() {group=SimpleSlurp, parameterName=SS_Color01, sortPriority=0, defaultValue=<rgba 0.1449 0.9271 0.8196 1>}
  n6 = ScalarParameter() {defaultValue=1, group=SimpleSlurp, parameterName=SS_Color01Multiplier, sortPriority=1}
  NamedRerouteDeclaration(None=LinearInterpolate(A=Multiply(A=@n3, B=@n4), B=Multiply(A=@n5.RGB, B=@n6), Alpha=NamedRerouteUsage())) {name=OUT_Color, nodeColor=<rgba 0 1 0.7188 1>}
  n7 = SphereMask(A=FunctionInput() {bUsePreviewValueAsDefault=true, inputName=preskinnedPosition}, B=Constant3Vector() {constant=<rgba 0 6 106 0>}) {attenuationRadius=40, hardnessPercent=0}
  NamedRerouteDeclaration(None=Multiply(A=@n7, B=@n7)) {name=BellyMask, nodeColor=<rgba 1 0 0.2344 1>}
  n8 = VectorParameter() {group=SimpleSlurp, parameterName=SS_DirGradientVector, sortPriority=40, defaultValue=<rgba 0 0 -1 0>}
  n9 = ConstantBiasScale(None=DotProduct(A=Normalize(VectorInput=@n8.RGB), B=PixelNormalWS()))
  n10 = ScalarParameter() {defaultValue=1, group=SimpleSlurp, parameterName=SS_DirGradientPow, sortPriority=41}
  NamedRerouteDeclaration(None=LinearInterpolate(
    B=Power(Base=@n9, Exp=@n10)
    Alpha=ScalarParameter() {defaultValue=0.75, group=SimpleSlurp, parameterName=SS_DirGradientStrength, sortPriority=42}
   ) {constA=1}) {name=Gradient, nodeColor=<rgba 1 0.3047 0 1>}
  n11 = ScalarParameter() {defaultValue=0.75, group=SimpleSlurp, parameterName=SS_FresnelExp_01, sortPriority=10}
  n12 = Power(
   Base=OneMinus(None=Fresnel(ExponentIn=@n11))
   Exp=ScalarParameter() {defaultValue=2, group=SimpleSlurp, parameterName=SS_FresnelPow_01, sortPriority=11}
  )
  n13 = Power(
   Base=Fresnel(ExponentIn=ScalarParameter() {defaultValue=0.75, group=SimpleSlurp, parameterName=SS_FresnelExp_02, sortPriority=12})
   Exp=ScalarParameter() {defaultValue=1, group=SimpleSlurp, parameterName=SS_FresnelPow_02, sortPriority=13}
  )
  NamedRerouteDeclaration(None=Multiply(A=Multiply(A=@n12, B=@n13), B=ScalarParameter() {defaultValue=3, group=SimpleSlurp, parameterName=SS_FresnelMultiplier, sortPriority=14})) {name=DoubleFresnel, nodeColor=<rgba 1 0 0.5156 1>}
  n14 = Multiply(A=AtmosphericLightVector()) {constB=-1}
  n15 = CameraVectorWS()
  n16 = Power(
   Base=ConstantBiasScale(None=DotProduct(A=@n14, B=@n15))
   Exp=ScalarParameter() {defaultValue=2, group=SimpleSlurp, parameterName=SS_SunScatterPow, sortPriority=31}
  )
  n17 = ScalarParameter() {defaultValue=3, group=SimpleSlurp, parameterName=SS_ScatterMulitplier}
  n18 = ScalarParameter() {defaultValue=0.5, group=SimpleSlurp, parameterName=SS_ScatterStrength, sortPriority=30}
  NamedRerouteDeclaration(None=LinearInterpolate(B=Multiply(A=@n16, B=@n17), Alpha=Saturate(None=@n18)) {constA=1}) {name=Scattering, nodeColor=<rgba 0 0.8359 1 1>}
  n19 = Multiply(A=NamedRerouteUsage(), B=NamedRerouteUsage()) {constA=1}
  n20 = NamedRerouteUsage()
  NamedRerouteDeclaration(None=Saturate(None=Multiply(A=@n19, B=@n20))) {name=ColorBlend, nodeColor=<rgba 1 0.8438 0 1>}
  VertexNormalWS()
 ]}
