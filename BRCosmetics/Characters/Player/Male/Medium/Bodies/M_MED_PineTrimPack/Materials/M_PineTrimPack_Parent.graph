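# Parent character material for the PineTrimPack male medium body cosmetic, built on the shared 2023 character PreFX and PostFX material functions with material attributes.
# PreFX output gets an emissive from a SimpleSlurp function masked by the blue channel of M, then passes through the PineTrimPack function.
# A static switch, PineTrimPack_UseOrangeSkin, selects that result by default or the plain PreFX output, bypassing both effects, before PostFX finishes it.
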Material /BRCosmetics/Characters/Player/Male/Medium/Bodies/M_MED_PineTrimPack/Materials/M_PineTrimPack_Parent
graph {
 properties {
  bUseMaterialAttributes=true
  bUsedWithMeshDeformer=true
  bUsedWithMorphTargets=true
  bUsedWithSkeletalMesh=true
  pixelDepthOffsetMode=PDOM_Legacy
  shadingModel=MSM_FromMaterialExpression
  usedShadingModels=Default Lit | Subsurface | Clear Coat | Subsurface Profile | Cloth
 }
 nodes [
  n1 = MaterialFunctionCall() {materialFunction=&MF_CharacterParent_2023_PreFX}
  n2 = SetMaterialAttributes(
   MaterialAttributes=@n1.Result
   Emissive Color=Multiply(
    A=MaterialFunctionCall!(preskinnedPosition=@n1.Pre skinned Local Position) {materialFunction=&/MF_SimpleSlurp_PineTrimPack}.Result
    B=ComponentMask(None=@n1.M) {b=true}
   )
  ) {attributeSetTypes [#1]}
  MaterialFunctionCall(
   Material=StaticSwitchParameter(
    True=MaterialFunctionCall!(PreSkinnedPosition=@n1.Pre skinned Local Position, Masks=@n1.Pre skinned Local Position, Fresnel=@n1.Pre skinned Local Position, In=@n2) {materialFunction=&/MF_PineTrimPack}.Result
    False=@n1.Result
   ) {defaultValue=true, parameterName=PineTrimPack_UseOrangeSkin}
   M=@n1.M
   Pre Skinned Local Position=@n1.M
   Distance Blend=@n1.M
   Fresnel=@n1.M
  ) {materialFunction=&MF_CharacterParent_2023_PostFX}
 ]
}
