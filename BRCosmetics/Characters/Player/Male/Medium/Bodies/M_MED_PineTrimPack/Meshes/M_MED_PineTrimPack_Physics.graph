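# Physics asset for the M_MED_PineTrimPack male medium player body cosmetic, setting up physics for its necklace.
# A chain of capsule bodies runs from spine_05 through base_necklace to dyn_necklace and dyn_necklace_01.
# The base is almost locked to the spine, while the two dynamic links swing within limited cones with no twist.
# A large, unconstrained necklace_collider capsule appears to serve as a collision volume for the necklace.
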PhysicsAsset /BRCosmetics/Characters/Player/Male/Medium/Bodies/M_MED_PineTrimPack/Meshes/M_MED_PineTrimPack_Physics
bodies {
 base_necklace [{shapeType=Capsule, length=0.505, radius=0.505, shapeName=base_necklace_capsule, rotation=<rot 0 -90 0>}]
 dyn_necklace [{
   shapeType=Capsule
   length=6.7674
   radius=2.7698
   shapeName=dyn_necklace_capsule
   center=<1.6913 -2.0998 -0.2685>
   rotation=<rot -0.3587 -90 3.1985>
  }]
 dyn_necklace_01 [{
   shapeType=Capsule
   length=4.959
   radius=2.1854
   shapeName=dyn_necklace_01_capsule
   center=<-0.1702 -1.3933 0.0927>
   rotation=<rot -0.3948 -90 -3.3303>
  }]
 necklace_collider [{shapeType=Capsule, length=13, radius=13, shapeName=necklace_collider_capsule, center=<-7.6885 -9.4164 0>, rotation=<rot 0 -90 0>}]
 spine_05 [{
   shapeType=Capsule
   length=10.2313
   radius=10.4931
   shapeName=spine_05_capsule
   center=<-2.605 0.3154 -0.3054>
   rotation=<rot 4.017 -90 -1.9128>
  }]
}
constraints [
 {bone1Name=dyn_necklace_01, bone2Name=dyn_necklace, swing1LimitDegrees=40, swing2LimitDegrees=20, twistLimitDegrees=0}
 {bone1Name=dyn_necklace, bone2Name=base_necklace, swing1LimitDegrees=20, swing2LimitDegrees=20, twistLimitDegrees=0}
 {bone1Name=base_necklace, bone2Name=spine_05, swing1LimitDegrees=1, swing2LimitDegrees=1, twistLimitDegrees=1}
]
