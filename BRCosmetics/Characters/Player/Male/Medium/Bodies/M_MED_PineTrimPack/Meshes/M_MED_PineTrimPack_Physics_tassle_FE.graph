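# Physics asset for the tassels on the PineTrimPack medium male player body cosmetic, covering the tassel bones, pelvis and thighs.
# A tiny base_tassles capsule anchors dyn_tassles, which forks into left and right three-segment capsule chains with limited swing and twist, widest swing1 at the tips.
# Pelvis and thigh capsules provide body shapes, with a second offset capsule on the left thigh.
# Constraints from thighs and tassel base to pelvis leave all angular motion free.
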PhysicsAsset /BRCosmetics/Characters/Player/Male/Medium/Bodies/M_MED_PineTrimPack/Meshes/M_MED_PineTrimPack_Physics_tassle_FE
bodies {
 base_tassles [{shapeType=Capsule, length=0.505, radius=0.505, shapeName=base_tassles_capsule, rotation=<rot 0 -90 0>}]
 dyn_tassle_01_l [{
   shapeType=Capsule
   length=5.8168
   radius=2.4
   shapeName=dyn_tassle_01_l_capsule
   center=<4.5755 -0.8176 1.2942>
   rotation=<rot 0 87.747 -85.1956>
  }]
 dyn_tassle_01_r [{
   shapeType=Capsule
   length=6.9011
   radius=2.4
   shapeName=dyn_tassle_01_r_capsule
   center=<4.3515 1.0614 1.2729>
   rotation=<rot 0 96.3248 -87.1503>
  }]
 dyn_tassle_02_l [{
   shapeType=Capsule
   length=4.5782
   radius=2.8
   shapeName=dyn_tassle_02_l_capsule
   center=<2.7356 -0.1982 0.0725>
   rotation=<rot 0 -95.0263 -89.538>
  }]
 dyn_tassle_02_r [{
   shapeType=Capsule
   length=2.5771
   radius=2.8
   shapeName=dyn_tassle_02_r_capsule
   center=<2.2407 -0.0397 -0.0137>
   rotation=<rot 0 -89.5768 -88.1996>
  }]
 dyn_tassle_03_l [{
   shapeType=Capsule
   length=5.0689
   radius=2.9112
   shapeName=dyn_tassle_03_l_capsule
   center=<2.916 0.0073 -0.2394>
   rotation=<rot 0 -83.4733 -84.0491>
  }]
 dyn_tassle_03_r [{
   shapeType=Capsule
   length=4.933
   radius=3.1076
   shapeName=dyn_tassle_03_r_capsule
   center=<3.0024 0.2179 -0.3222>
   rotation=<rot 0 -85.2406 -82.9725>
  }]
 dyn_tassles [{
   shapeType=Capsule
   length=4.2973
   radius=2.6684
   shapeName=dyn_tassles_capsule
   center=<10.2922 -0.9258 0.8076>
   rotation=<rot 0 90.8822 -88.5904>
  }]
 pelvis [{
   shapeType=Capsule
   length=14.6241
   radius=13.4057
   shapeName=pelvis_capsule
   center=<-6.2651 -0.6841 -0.5905>
   rotation=<rot 10.8797 -90 -3.9862>
  }]
 thigh_l [
  {shapeType=Capsule, length=40, radius=9.7206, shapeName=thigh_l_capsule, center=<-19.2367 0.1266 0.133>, rotation=<rot 90 -2.8125 0>}
  {shapeType=Capsule, length=40, radius=10, shapeName=thigh_l_capsule, center=<-17.4562 0.6952 -16.2663>, rotation=<rot 90 0 0>}
 ]
 thigh_r [{shapeType=Capsule, length=40, radius=10.0925, shapeName=thigh_r_capsule, center=<19.354 0.8717 -0.4895>, rotation=<rot 90 0 0>}]
}
constraints [
 {bone1Name=dyn_tassle_01_l, bone2Name=dyn_tassles, swing1LimitDegrees=20, swing2LimitDegrees=40, twistLimitDegrees=10}
 {bone1Name=dyn_tassle_02_l, bone2Name=dyn_tassle_01_l, swing1LimitDegrees=20, swing2LimitDegrees=40, twistLimitDegrees=10}
 {bone1Name=dyn_tassle_03_l, bone2Name=dyn_tassle_02_l, swing1LimitDegrees=45, swing2LimitDegrees=45, twistLimitDegrees=10}
 {bone1Name=dyn_tassle_01_r, bone2Name=dyn_tassles, swing1LimitDegrees=20, swing2LimitDegrees=40, twistLimitDegrees=10}
 {bone1Name=dyn_tassle_02_r, bone2Name=dyn_tassle_01_r, swing1LimitDegrees=20, swing2LimitDegrees=40, twistLimitDegrees=10}
 {bone1Name=dyn_tassle_03_r, bone2Name=dyn_tassle_02_r, swing1LimitDegrees=45, swing2LimitDegrees=45, twistLimitDegrees=10}
 {bone1Name=dyn_tassles, bone2Name=base_tassles, swing1LimitDegrees=45, swing2LimitDegrees=60, twistLimitDegrees=20}
 {
  bone1Name=thigh_l
  bone2Name=pelvis
  swing1LimitDegrees=45
  swing1Motion=Free
  swing2LimitDegrees=45
  swing2Motion=Free
  twistLimitDegrees=45
  twistMotion=Free
 }
 {
  bone1Name=thigh_r
  bone2Name=pelvis
  swing1LimitDegrees=45
  swing1Motion=Free
  swing2LimitDegrees=45
  swing2Motion=Free
  twistLimitDegrees=45
  twistMotion=Free
 }
 {
  bone1Name=base_tassles
  bone2Name=pelvis
  swing1LimitDegrees=45
  swing1Motion=Free
  swing2LimitDegrees=45
  swing2Motion=Free
  twistLimitDegrees=45
  twistMotion=Free
 }
]
